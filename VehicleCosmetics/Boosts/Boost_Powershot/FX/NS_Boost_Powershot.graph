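# Niagara system for the Powershot vehicle boost cosmetic, trailing ribbons, lightning, embers and swirling sparks from vehicle boost sockets.
# Each emitter uses a filtered socket count to spread four ribbon layers and three sprite layers across skeletal mesh sockets.
# Ribbons and embers inherit decaying source movement, while sparks swirl through vortex and curl-noise forces.
# User flags apply paint color to some layers and retune motion, lifetime or size for frontend preview.
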NiagaraSystem /VehicleCosmetics/Boosts/Boost_Powershot/FX/NS_Boost_Powershot
dependencies {
 usedDataInterfaces [
  &NiagaraDataInterfaceCurve
  &NiagaraDataInterfaceVectorCurve
  &NiagaraDataInterfaceSpriteRendererInfo
  &NiagaraDataInterfaceSkeletalMesh
 ]
 usedDynamicInputs []
 usedModules [
  &self:SystemSpawnScript.SystemScriptSource.SystemScriptGraph.NiagaraNodeAssignment_0.SetVariables_#1
  &SystemState
 ]
}
emitters {
 NE_VC_Base_Sparks_01 {
  emitterUpdateScript [
   &EmitterState(Life Cycle Mode=::NewEnumerator1, Loop Duration=1, UseLoopDelay=false)
   &VC_FilteredSocketCount()
   &SpawnRate(SpawnRate=40, Use Spawn Probability=false)
  ]
  particleSpawnScript [
   &VC_PaintedVariantsOverride() {enabled=false}
   &InitializeParticle(
    Color=<rgba 2 2 2 5>
    Lifetime=&RandomRangeFloat(Minimum=0.5, Maximum=0.7)
    Sprite Rotation Mode=::NewEnumerator2
    Sprite Size Max=<5 6>
    Sprite Size Min=<4 5>
    Sprite Size Mode=::NewEnumerator2
    UsePositionOffset=false
    Write Material Random=true
   )
   &PartitionParticles(Number Of Partitions=$Emitter.PortCount)
   &ShapeLocation(MaskSpawnGroup=false, Offset=<100 0 0>, Offset Mode=::NewEnumerator0, Override Seed=false, Sphere Radius=200, UseOwnerScale=true) {enabled=false}
   &SkeletalMeshLocation(
    Bone / Socket Index=$Particles.PartitionParticles.Partition
    Bone / Socket Sampling Mode=::NewEnumerator5
    MaskSpawnGroup=false
    Mesh Sampling Type=::NewEnumerator6
    Skeletal Mesh=$System.SkeletalMesh
   )
   &Color(Color=&MakeCustomLinearColorFromBool(Bool=$User.bIsPainted, True Color=$User.PaintedColor, False Color=<rgba 2 0.1 0.05 5>))
   &AddVelocity(Velocity=&MakeCustomVectorFromBool(
     Bool=$User.bIsFrontendPreview
     True Vector=&RandomRangeVector(Minimum=<-150 -700 -500>, Maximum=<-300 700 500>)
     False Vector=&RandomRangeVector(Minimum=<-150 -500 -100>, Maximum=<-150 500 100>)
    ))
   &AddVelocity(
    Rotation Coordinate Space=::World
    Velocity=&RandomRangeVector(Minimum=<0 -30 -30>, Maximum=<0 30 30>)
    Velocity Speed Scale=3
   ) {enabled=false, module=AddVelocity001}
  ]
  particleUpdateScript [
   &ParticleState()
   &SpriteRotationRate(Rotation Rate=&RandomRangeFloat(Minimum=-10, Maximum=10))
   &InheritSourceMovement(
    Applied Position Scale=&Multiply_VectorByFloat(
     Vector=<1 1 1>
     Float=&FloatFromCurve(FloatCurve=DataInterface() {propertyValues {Curve {keys [
          {
           arriveTangentWeight=0.4252
           interpMode=RCIM_Cubic
           leaveTangentWeight=0.4252
           tangentMode=RCTM_User
           tangentWeightMode=RCTWM_WeightedBoth
           time=0.05
           value=1
          }
          {interpMode=RCIM_Cubic, tangentMode=RCTM_User, time=1, value=0}
         ]}}})
    )
    ApplyForce=false
    ApplyPosition=true
    ApplyPositionFalloff=false
   ) {enabled=false}
   &AccelerationForce(Acceleration=&RandomRangeVector(Minimum=<-500 0 0>, Maximum=<-1000 0 0>)) {enabled=false}
   &AccelerationForce(Acceleration=&MakeCustomVectorFromBool(
     Bool=$User.bIsFrontendPreview
     True Vector=&Multiply_VectorByFloat(
      Vector=&RandomRangeVector(Minimum=<0 -200 -150>, Maximum=<0 200 150>)
      Float=&Multiply_Float(
       A=&FloatFromCurve(
        FloatCurve=DataInterface() {propertyValues {Curve {keys [<at 0 0 RCIM_Cubic>, {interpMode=RCIM_Cubic, tangentMode=RCTM_User, time=0.25, value=2}, {tangentMode=RCTM_User, time=0.75, value=-2}]}}}
        Scale Curve=4
       )
       B=1
      )
     )
     False Vector=&Multiply_VectorByFloat(
      Vector=&RandomRangeVector(Minimum=<0 -200 -150>, Maximum=<0 200 150>)
      Float=&Multiply_Float(
       A=&FloatFromCurve(
        FloatCurve=DataInterface() {propertyValues {Curve {keys [<at 0 0 RCIM_Cubic>, {interpMode=RCIM_Cubic, tangentMode=RCTM_User, time=0.25, value=2}, {tangentMode=RCTM_User, time=0.75, value=-2}]}}}
        Scale Curve=4
       )
       B=1
      )
     )
    )) {module=AccelerationForce001}
   &CurlNoiseForce(Noise Strength=200, Pan Noise=true, Pan Noise Field=<1 0 0>)
   &RotateAroundPoint(
    Derive the Rotation Axis=false
    Radius:NiagaraFloat=&FloatFromCurve(
     FloatCurve=DataInterface() {propertyValues {Curve {keys [
         {interpMode=RCIM_Cubic, tangentMode=RCTM_User, time=0, value=0}
         {interpMode=RCIM_Cubic, tangentMode=RCTM_User, time=0.5, value=1}
         {tangentMode=RCTM_User, time=1, value=0}
        ]}}}
     Scale Curve=50
    )
    Rotation Axis:Vector3f=&RandomRangeVector(Minimum=<-1 -1 -1>, Maximum=<1 1 1>)
    Rotation Center:NiagaraPosition=&AddVectorToPosition(Position=$Particles.Initial.SkeletalMeshLocation.SampledPosition, Vector=<0 0 0>)
    Rotation Rate:NiagaraFloat=&FloatFromCurve(FloatCurve=DataInterface() {propertyValues {Curve {keys [{tangentMode=RCTM_User, time=0, value=0}, {tangentMode=RCTM_User, time=1, value=1}]}}})
    X Axis:Vector3f=<1 1 1>
   ) {enabled=false}
   &VortexForce(
    Origin Pull Amount:NiagaraFloat=&MakeCustomFloatFromBool(
     Bool=$User.bIsFrontendPreview
     True Float=&RandomRangeFloat(Minimum=5000, Maximum=8000, Evaluation Type=::NewEnumerator1)
     False Float=&RandomRangeFloat(Minimum=10000, Maximum=10000, Evaluation Type=::NewEnumerator1)
    )
    UseOriginOffset=true
    Vortex Axis=&RandomRangeVector(Minimum=<1 0 0>, Maximum=<-1 0 0>, Evaluation Type=::NewEnumerator1)
    Vortex Axis Coordinate Space:ENiagaraCoordinateSpace=::Local
    Vortex Force Amount=&MakeCustomFloatFromBool(
     Bool=$User.bIsFrontendPreview
     True Float=&RandomRangeFloat(Minimum=3000, Maximum=5000)
     False Float=&RandomRangeFloat(Minimum=7000, Maximum=10000)
    )
    Vortex Origin:NiagaraPosition=$Particles.Initial.Position
    Vortex Origin Offset:Vector3f=&RandomRangeVector(Minimum=<0 0 0>, Maximum=<-20 0 0>, Evaluation Type=::NewEnumerator1)
    Vortex Origin Offset Coordinate Space:ENiagaraCoordinateSpace=::Local
   )
   &SolveForcesAndVelocity(Clamp Velocity=false, Limit Acceleration=false)
   &ScaleColor(
    Scale RGB=&MakeCustomVectorFromBool(
     Bool=$User.bIsPainted
     True Vector=&Multiply_VectorByFloat(Vector=<1 1 1>, Float=10)
     False Vector=&Multiply_VectorByFloat(Vector=<15 8 8>, Float=10)
    )
    ScaleA=true
    ScaleRGB=true
   )
   &ScaleSpriteSize(Uniform Curve Sprite Scale=DataInterface() {propertyValues {Curve {keys [
        {arriveTangent=0.3218, interpMode=RCIM_Cubic, leaveTangent=0.3218, tangentMode=RCTM_User, time=0.25, value=1}
        <at 0.6 2 RCIM_Cubic>
        <at 1 1 RCIM_Cubic>
       ]}}})
   &SubUVAnimation(
    "Sprite Renderer "=DataInterface() {propertyValues {SpriteRenderer=&self:NE_VC_Base_Sparks_01_0.NiagaraSpriteRendererProperties_0}}
    UseEndFrame=false
    UseStartFrame=false
   )
   &DynamicMaterialParameters(
    Index 0 Param 1=2
    Module.Index0Param0Write=true
    Module.Index0Param1Write=true
    Module.Index0Param2Write=true
    Module.Index0Param3Write=true
   )
  ]
  rendererClasses [&NiagaraSpriteRendererProperties]
  renderers [{rendererClass=&NiagaraSpriteRendererProperties}]
 }
 NE_VC_Base_Sparks_01001 {
  emitterUpdateScript [
   &EmitterState(Life Cycle Mode=::NewEnumerator1, Loop Duration=1, UseLoopDelay=false)
   &VC_FilteredSocketCount()
   &SpawnRate(SpawnRate=40, Use Spawn Probability=false)
  ]
  particleSpawnScript [
   &VC_PaintedVariantsOverride() {enabled=false}
   &InitializeParticle(
    Color=<rgba 2 2 2 5>
    Lifetime=&RandomRangeFloat(Minimum=0.5, Maximum=0.7)
    Sprite Rotation Mode=::NewEnumerator2
    Sprite Size Max=<5 6>
    Sprite Size Min=<4 5>
    Sprite Size Mode=::NewEnumerator2
    UsePositionOffset=false
    Write Material Random=true
   )
   &PartitionParticles(Number Of Partitions=$Emitter.PortCount)
   &ShapeLocation(MaskSpawnGroup=false, Offset=<100 0 0>, Offset Mode=::NewEnumerator0, Override Seed=false, Sphere Radius=200, UseOwnerScale=true) {enabled=false}
   &SkeletalMeshLocation(
    Bone / Socket Index=$Particles.PartitionParticles.Partition
    Bone / Socket Sampling Mode=::NewEnumerator5
    MaskSpawnGroup=false
    Mesh Sampling Type=::NewEnumerator6
    Skeletal Mesh=$System.SkeletalMesh
   )
   &Color(Color=<rgba 0.2 0.05 1 1>)
   &AddVelocity(Velocity=&MakeCustomVectorFromBool(
     Bool=$User.bIsFrontendPreview
     True Vector=&RandomRangeVector(Minimum=<-150 -700 -500>, Maximum=<-300 700 500>)
     False Vector=&RandomRangeVector(Minimum=<-150 -500 -100>, Maximum=<-150 500 100>)
    ))
   &AddVelocity(
    Rotation Coordinate Space=::World
    Velocity=&RandomRangeVector(Minimum=<0 -30 -30>, Maximum=<0 30 30>)
    Velocity Speed Scale=3
   ) {enabled=false, module=AddVelocity001}
  ]
  particleUpdateScript [
   &ParticleState()
   &SpriteRotationRate(Rotation Rate=&RandomRangeFloat(Minimum=-10, Maximum=10))
   &InheritSourceMovement(
    Applied Position Scale=&Multiply_VectorByFloat(
     Vector=<1 1 1>
     Float=&FloatFromCurve(FloatCurve=DataInterface() {propertyValues {Curve {keys [
          {
           arriveTangentWeight=0.4252
           interpMode=RCIM_Cubic
           leaveTangentWeight=0.4252
           tangentMode=RCTM_User
           tangentWeightMode=RCTWM_WeightedBoth
           time=0.05
           value=1
          }
          {interpMode=RCIM_Cubic, tangentMode=RCTM_User, time=1, value=0}
         ]}}})
    )
    ApplyForce=false
    ApplyPosition=true
    ApplyPositionFalloff=false
   ) {enabled=false}
   &AccelerationForce(Acceleration=&RandomRangeVector(Minimum=<-500 0 0>, Maximum=<-1000 0 0>)) {enabled=false}
   &AccelerationForce(Acceleration=&MakeCustomVectorFromBool(
     Bool=$User.bIsFrontendPreview
     True Vector=&Multiply_VectorByFloat(
      Vector=&RandomRangeVector(Minimum=<0 -200 -150>, Maximum=<0 200 150>)
      Float=&Multiply_Float(
       A=&FloatFromCurve(
        FloatCurve=DataInterface() {propertyValues {Curve {keys [<at 0 0 RCIM_Cubic>, {interpMode=RCIM_Cubic, tangentMode=RCTM_User, time=0.25, value=2}, {tangentMode=RCTM_User, time=0.75, value=-2}]}}}
        Scale Curve=4
       )
       B=1
      )
     )
     False Vector=&Multiply_VectorByFloat(
      Vector=&RandomRangeVector(Minimum=<0 -200 -150>, Maximum=<0 200 150>)
      Float=&Multiply_Float(
       A=&FloatFromCurve(
        FloatCurve=DataInterface() {propertyValues {Curve {keys [<at 0 0 RCIM_Cubic>, {interpMode=RCIM_Cubic, tangentMode=RCTM_User, time=0.25, value=2}, {tangentMode=RCTM_User, time=0.75, value=-2}]}}}
        Scale Curve=4
       )
       B=1
      )
     )
    )) {module=AccelerationForce001}
   &CurlNoiseForce(Noise Strength=200, Pan Noise=true, Pan Noise Field=<1 0 0>)
   &RotateAroundPoint(
    Derive the Rotation Axis=false
    Radius:NiagaraFloat=&FloatFromCurve(
     FloatCurve=DataInterface() {propertyValues {Curve {keys [
         {interpMode=RCIM_Cubic, tangentMode=RCTM_User, time=0, value=0}
         {interpMode=RCIM_Cubic, tangentMode=RCTM_User, time=0.5, value=1}
         {tangentMode=RCTM_User, time=1, value=0}
        ]}}}
     Scale Curve=50
    )
    Rotation Axis:Vector3f=&RandomRangeVector(Minimum=<-1 -1 -1>, Maximum=<1 1 1>)
    Rotation Center:NiagaraPosition=&AddVectorToPosition(Position=$Particles.Initial.SkeletalMeshLocation.SampledPosition, Vector=<0 0 0>)
    Rotation Rate:NiagaraFloat=&FloatFromCurve(FloatCurve=DataInterface() {propertyValues {Curve {keys [{tangentMode=RCTM_User, time=0, value=0}, {tangentMode=RCTM_User, time=1, value=1}]}}})
    X Axis:Vector3f=<1 1 1>
   ) {enabled=false}
   &VortexForce(
    Origin Pull Amount:NiagaraFloat=&MakeCustomFloatFromBool(
     Bool=$User.bIsFrontendPreview
     True Float=&RandomRangeFloat(Minimum=5000, Maximum=8000, Evaluation Type=::NewEnumerator1)
     False Float=&RandomRangeFloat(Minimum=10000, Maximum=10000, Evaluation Type=::NewEnumerator1)
    )
    UseOriginOffset=true
    Vortex Axis=&RandomRangeVector(Minimum=<1 0 0>, Maximum=<-1 0 0>, Evaluation Type=::NewEnumerator1)
    Vortex Axis Coordinate Space:ENiagaraCoordinateSpace=::Local
    Vortex Force Amount=&MakeCustomFloatFromBool(
     Bool=$User.bIsFrontendPreview
     True Float=&RandomRangeFloat(Minimum=3000, Maximum=5000)
     False Float=&RandomRangeFloat(Minimum=7000, Maximum=10000)
    )
    Vortex Origin:NiagaraPosition=$Particles.Initial.Position
    Vortex Origin Offset:Vector3f=&RandomRangeVector(Minimum=<0 0 0>, Maximum=<-20 0 0>, Evaluation Type=::NewEnumerator1)
    Vortex Origin Offset Coordinate Space:ENiagaraCoordinateSpace=::Local
   )
   &SolveForcesAndVelocity(Clamp Velocity=false, Limit Acceleration=false)
   &ScaleColor(
    Scale RGB=&MakeCustomVectorFromBool(
     Bool=$User.bIsPainted
     True Vector=&Multiply_VectorByFloat(Vector=<1 1 1>, Float=5)
     False Vector=&Multiply_VectorByFloat(Vector=<1 1 1>, Float=150)
    )
    ScaleA=true
    ScaleRGB=true
   )
   &ScaleSpriteSize(Uniform Curve Sprite Scale=DataInterface() {propertyValues {Curve {keys [
        {arriveTangent=0.3218, interpMode=RCIM_Cubic, leaveTangent=0.3218, tangentMode=RCTM_User, time=0.25, value=1}
        <at 0.6 2 RCIM_Cubic>
        <at 1 1 RCIM_Cubic>
       ]}}})
   &SubUVAnimation(
    "Sprite Renderer "=DataInterface() {propertyValues {SpriteRenderer=&self:NE_VC_Base_Sparks_01001_1.NiagaraSpriteRendererProperties_0}}
    UseEndFrame=false
    UseStartFrame=false
   )
   &DynamicMaterialParameters(
    Index 0 Param 1=2
    Module.Index0Param0Write=true
    Module.Index0Param1Write=true
    Module.Index0Param2Write=true
    Module.Index0Param3Write=true
   )
  ]
  rendererClasses [&NiagaraSpriteRendererProperties]
  renderers [{rendererClass=&NiagaraSpriteRendererProperties}]
 }
 NE_VC_CoreLightningTrail001 {
  emitterUpdateScript [
   &EmitterState(Life Cycle Mode=::NewEnumerator1, Loop Duration=1, UseLoopDelay=false)
   &VC_FilteredSocketCount()
   &VC_SpawnRibbon_v1(Number of Ribbons=$Emitter.PortCount, SpawnRate=50)
  ]
  particleSpawnScript [
   &VC_PaintedVariantsOverride() {enabled=false}
   &VC_InitializeRibbon_v1(
    Enable Skeletal Position=true
    Lifetime=&MakeCustomFloatFromBool(Bool=$User.bIsFrontendPreview, True Float=1.4, False Float=1)
    Offset Amount=20
    Offset Axis:ENiagaraOrientationAxis=::XAxis
    Ribbon Width=25
   )
   &AddVelocity(Velocity=<-1000 0 0>) {module=AddVelocity001}
   &Color(Color=&MakeCustomLinearColorFromBool(Bool=$User.bIsPainted, True Color=$User.PaintedColor, False Color=<rgba 10 0.6563 0.25 5>))
  ]
  particleUpdateScript [
   &DynamicMaterialParameters(
    Index 0 Param 1=&RandomRangeFloat(Minimum=2, Maximum=5)
    Index 0 Param 2=&RandomRangeFloat(Minimum=0, Maximum=1)
    Index 0 Param 3=0.1
    Index 0 Param 4=0.3
    Module.Index0Param0Write=true
    Module.Index0Param1Write=true
    Module.Index0Param2Write=true
    Module.Index0Param3Write=true
   )
   &ParticleState()
   &SolveForcesAndVelocity(Clamp Velocity=false, Limit Acceleration=false) {enabled=false}
   &InheritSourceMovement(
    Applied Position Scale=&Multiply_VectorByFloat(
     Vector=<1 1 1>
     Float=&FloatFromCurve(FloatCurve=DataInterface() {propertyValues {Curve {keys [
          {
           arriveTangentWeight=0.4252
           interpMode=RCIM_Cubic
           leaveTangentWeight=0.4252
           tangentMode=RCTM_User
           tangentWeightMode=RCTWM_WeightedBoth
           time=0.05
           value=1
          }
          {interpMode=RCIM_Cubic, tangentMode=RCTM_User, time=1, value=0}
         ]}}})
    )
    ApplyForce=false
    ApplyPosition=true
    ApplyPositionFalloff=false
   )
   &SolveForcesAndVelocity(Clamp Velocity=false, Limit Acceleration=true) {module=SolveForcesAndVelocity001}
   &VC_UpdateRibbon_v1(Enable Skeletal Position=true)
   &ScaleColor(
    Scale Alpha=&FloatFromCurve(FloatCurve=DataInterface() {propertyValues {Curve {keys [
         {interpMode=RCIM_Cubic, tangentMode=RCTM_Break, time=0, value=5}
         {interpMode=RCIM_Cubic, tangentMode=RCTM_Break, time=0.25, value=1}
         {interpMode=RCIM_Cubic, tangentMode=RCTM_Break, time=1, value=0}
        ]}}})
    Scale RGB=&Multiply_VectorByFloat(
     Vector=&VectorFromCurve(VectorCurve=DataInterface() {propertyValues {
        XCurve {keys [{interpMode=RCIM_Cubic, tangentMode=RCTM_User, time=0, value=1}, {interpMode=RCIM_Cubic, tangentMode=RCTM_User, time=1, value=0}]}
        YCurve {keys [{interpMode=RCIM_Cubic, tangentMode=RCTM_User, time=0, value=1}, {interpMode=RCIM_Cubic, tangentMode=RCTM_User, time=1, value=0}]}
        ZCurve {keys [{interpMode=RCIM_Cubic, tangentMode=RCTM_User, time=0, value=1}, {interpMode=RCIM_Cubic, tangentMode=RCTM_User, time=1, value=0}]}
       }})
     Float=&MakeCustomFloatFromBool(Bool=$User.bIsPainted, True Float=5, False Float=1)
    )
    ScaleA=true
    ScaleRGB=true
   )
   &ScaleRibbonWidth(Ribbon Width Scale=&FloatFromCurve(FloatCurve=DataInterface() {propertyValues {Curve {keys [{interpMode=RCIM_Cubic, tangentMode=RCTM_User, time=0, value=3}, {interpMode=RCIM_Cubic, tangentMode=RCTM_User, time=1, value=1}]}}}))
  ]
  rendererClasses [&NiagaraRibbonRendererProperties]
  renderers [{rendererClass=&NiagaraRibbonRendererProperties}]
 }
 NE_VC_CoreTrail {
  emitterUpdateScript [
   &EmitterState(Life Cycle Mode=::NewEnumerator1, Loop Duration=1, UseLoopDelay=false)
   &VC_FilteredSocketCount()
   &VC_SpawnRibbon_v1(Number of Ribbons=$Emitter.PortCount, SpawnRate=50)
  ]
  particleSpawnScript [
   &VC_PaintedVariantsOverride() {enabled=false}
   &VC_InitializeRibbon_v1(
    Enable Skeletal Position=true
    Lifetime=&MakeCustomFloatFromBool(Bool=$User.bIsFrontendPreview, True Float=2, False Float=1)
    Offset Amount:NiagaraFloat=&MakeCustomFloatFromBool(Bool=$User.bIsFrontendPreview, True Float=20, False Float=20)
    Offset Axis:ENiagaraOrientationAxis=::XAxis
    Ribbon Width=120
   )
   &AddVelocity(Velocity=<-1000 0 0>) {module=AddVelocity001}
   &Color(Color=&MakeCustomLinearColorFromBool(Bool=$User.bIsPainted, True Color=$User.PaintedColor, False Color=<rgba 5.5556 0.5556 10 5>))
  ]
  particleUpdateScript [
   &DynamicMaterialParameters(
    Index 0 Param 1=3
    Index 0 Param 2=&RandomRangeFloat(Minimum=0, Maximum=1)
    Index 0 Param 4=0.5
    Module.Index0Param0Write=true
    Module.Index0Param1Write=true
    Module.Index0Param2Write=true
    Module.Index0Param3Write=true
   )
   &ParticleState()
   &SolveForcesAndVelocity(Clamp Velocity=false, Limit Acceleration=false) {enabled=false}
   &InheritSourceMovement(
    Applied Position Scale=&Multiply_VectorByFloat(
     Vector=<1 1 1>
     Float=&FloatFromCurve(FloatCurve=DataInterface() {propertyValues {Curve {keys [
          {
           arriveTangentWeight=0.4252
           interpMode=RCIM_Cubic
           leaveTangentWeight=0.4252
           tangentMode=RCTM_User
           tangentWeightMode=RCTWM_WeightedBoth
           time=0.05
           value=1
          }
          {interpMode=RCIM_Cubic, tangentMode=RCTM_User, time=1, value=0}
         ]}}})
    )
    ApplyForce=false
    ApplyPosition=true
    ApplyPositionFalloff=false
   )
   &SolveForcesAndVelocity(Clamp Velocity=false, Limit Acceleration=true) {module=SolveForcesAndVelocity001}
   &VC_UpdateRibbon_v1(Enable Skeletal Position=true)
   &ScaleColor(
    Scale Alpha=&FloatFromCurve(FloatCurve=DataInterface() {propertyValues {Curve {keys [
         {interpMode=RCIM_Cubic, tangentMode=RCTM_Break, time=0, value=5}
         {interpMode=RCIM_Cubic, tangentMode=RCTM_Break, time=0.5528, value=1}
         {interpMode=RCIM_Cubic, tangentMode=RCTM_Break, time=0.9909, value=0}
        ]}}})
    Scale RGB=&VectorFromCurve(VectorCurve=DataInterface() {propertyValues {
       XCurve {keys [{interpMode=RCIM_Cubic, tangentMode=RCTM_Break, time=0, value=0.05}, <at 0.25 0.01>]}
       YCurve {keys [{interpMode=RCIM_Cubic, tangentMode=RCTM_Break, time=0, value=0.01}, <at 0.25 0.02>]}
       ZCurve {keys [{interpMode=RCIM_Cubic, tangentMode=RCTM_Break, time=0, value=0.1}, <at 0.25 0.025>]}
      }})
    ScaleA=true
    ScaleRGB=true
   )
   &ScaleRibbonWidth(Ribbon Width Scale=&FloatFromCurve(FloatCurve=DataInterface() {propertyValues {Curve {keys [
         <at 0 1 RCIM_Cubic>
         {arriveTangent=-0.642, interpMode=RCIM_Cubic, leaveTangent=-0.642, tangentMode=RCTM_User, time=0.3, value=0.85}
         {interpMode=RCIM_Cubic, tangentMode=RCTM_User, time=1, value=0}
        ]}}}))
  ]
  rendererClasses [&NiagaraRibbonRendererProperties]
  renderers [{rendererClass=&NiagaraRibbonRendererProperties}]
 }
 NE_VC_Embers {
  emitterUpdateScript [
   &EmitterState(Life Cycle Mode=::NewEnumerator1, Loop Duration=1, UseLoopDelay=false)
   &VC_FilteredSocketCount()
   &SpawnRate(SpawnRate=200, Use Spawn Probability=false)
  ]
  particleSpawnScript [
   &VC_PaintedVariantsOverride() {enabled=false}
   &InitializeParticle(
    Color=<rgba 2 2 2 5>
    Lifetime=&RandomRangeFloat(Minimum=0.1, Maximum=0.5)
    Sprite Rotation Mode=::NewEnumerator2
    Sprite Size Max=&MakeCustomVector2DFromBool(Bool=$User.bIsFrontendPreview, True Vector2D=<1 5>, False Vector2D=<1.5 6>)
    Sprite Size Min=&MakeCustomVector2DFromBool(Bool=$User.bIsFrontendPreview, True Vector2D=<0.5 3>, False Vector2D=<1 4>)
    Sprite Size Mode=::NewEnumerator2
    UsePositionOffset=false
    Write Material Random=true
   )
   &PartitionParticles(Number Of Partitions=$Emitter.PortCount)
   &ShapeLocation(MaskSpawnGroup=false, Offset=<-140 0 0>, Offset Mode=::NewEnumerator0, Override Seed=false, Sphere Radius=10, UseOwnerScale=true)
   &SkeletalMeshLocation(
    Bone / Socket Index=$Particles.PartitionParticles.Partition
    Bone / Socket Sampling Mode=::NewEnumerator5
    MaskSpawnGroup=false
    Mesh Sampling Type=::NewEnumerator6
    Skeletal Mesh=$System.SkeletalMesh
   )
   &Color(Color=&MakeCustomLinearColorFromBool(Bool=$User.bIsPainted, True Color=$User.PaintedColor, False Color=<rgba 2 0.1 0.05 5>))
   &AddVelocity(Velocity=<-1200 0 0>)
   &AddVelocity(
    Rotation Coordinate Space=::World
    Velocity=&RandomRangeVector(Minimum=<0 -30 -30>, Maximum=<0 30 30>)
    Velocity Speed Scale=3
   ) {module=AddVelocity001}
  ]
  particleUpdateScript [
   &ParticleState()
   &SpriteRotationRate(Rotation Rate=&RandomRangeFloat(Minimum=-10, Maximum=10))
   &InheritSourceMovement(
    Applied Position Scale=&Multiply_VectorByFloat(
     Vector=<1 1 1>
     Float=&FloatFromCurve(FloatCurve=DataInterface() {propertyValues {Curve {keys [
          {
           arriveTangentWeight=0.4252
           interpMode=RCIM_Cubic
           leaveTangentWeight=0.4252
           tangentMode=RCTM_User
           tangentWeightMode=RCTWM_WeightedBoth
           time=0.05
           value=1
          }
          {interpMode=RCIM_Cubic, tangentMode=RCTM_User, time=1, value=0}
         ]}}})
    )
    ApplyForce=false
    ApplyPosition=true
    ApplyPositionFalloff=false
   )
   &CurlNoiseForce(Noise Strength=500, Pan Noise=true, Pan Noise Field=<100 0 0>)
   &Drag(Use Linear Drag=true, Use Rotational Drag=true)
   &SolveForcesAndVelocity(Clamp Velocity=false, Limit Acceleration=false)
   &ScaleColor(
    Scale RGB=&MakeCustomVectorFromBool(Bool=$User.bIsPainted, True Vector=<1 1 1>, False Vector=<15 8 8>)
    ScaleA=true
    ScaleRGB=true
   )
   &ScaleSpriteSize(
    Uniform Curve Scale=2
    Uniform Curve Sprite Scale=DataInterface() {propertyValues {Curve {keys [
        {arriveTangent=0.3218, interpMode=RCIM_Cubic, leaveTangent=0.3218, tangentMode=RCTM_User, time=0.25, value=0}
        <at 0.75 1 RCIM_Cubic>
        <at 1 0 RCIM_Cubic>
       ]}}}
   )
  ]
  rendererClasses [&NiagaraSpriteRendererProperties]
  renderers [{rendererClass=&NiagaraSpriteRendererProperties}]
 }
 NE_VC_Lightning_01 {
  emitterUpdateScript [
   &EmitterState(Life Cycle Mode=::NewEnumerator1, Loop Duration=1, UseLoopDelay=false)
   &VC_FilteredSocketCount()
   &VC_SpawnRibbon_v1(Number of Ribbons=$Emitter.PortCount, SpawnRate=50)
  ]
  particleSpawnScript [
   &VC_PaintedVariantsOverride() {enabled=false}
   &VC_InitializeRibbon_v1(
    Enable Skeletal Position=true
    Lifetime=&MakeCustomFloatFromBool(Bool=$User.bIsFrontendPreview, True Float=0.8, False Float=1)
    Offset Amount:NiagaraFloat=&MakeCustomFloatFromBool(Bool=$User.bIsFrontendPreview, True Float=40, False Float=20)
    Offset Axis:ENiagaraOrientationAxis=::XAxis
    Ribbon Width=50
   )
   &AddVelocity(Velocity=&MakeCustomVectorFromBool(Bool=$User.bIsFrontendPreview, True Vector=<-2000 0 1>, False Vector=<-1000 0 0>)) {module=AddVelocity001}
   &Color(Color=<rgba 1 3 6 1>)
  ]
  particleUpdateScript [
   &ParticleState()
   &InheritSourceMovement(
    Applied Position Scale=&Multiply_VectorByFloat(
     Vector=<1 1 1>
     Float=&FloatFromCurve(FloatCurve=DataInterface() {propertyValues {Curve {keys [
          {
           arriveTangentWeight=0.4252
           interpMode=RCIM_Cubic
           leaveTangentWeight=0.4252
           tangentMode=RCTM_User
           tangentWeightMode=RCTWM_WeightedBoth
           time=0.05
           value=1
          }
          {interpMode=RCIM_Cubic, tangentMode=RCTM_User, time=1, value=0}
         ]}}})
    )
    ApplyForce=false
    ApplyPosition=true
    ApplyPositionFalloff=false
   )
   &SolveForcesAndVelocity(Clamp Velocity=false, Limit Acceleration=false)
   &VC_UpdateRibbon_v1(Enable Skeletal Position=true)
   &DynamicMaterialParameters(
    Index 0 Param 1=0.2
    Index 0 Param 2=20
    Module.Index0Param0Write=true
    Module.Index0Param1Write=true
    Module.Index0Param2Write=true
    Module.Index0Param3Write=true
   )
   &ScaleColor(
    Scale Alpha=&FloatFromCurve(FloatCurve=DataInterface() {propertyValues {Curve {keys [{interpMode=RCIM_Cubic, tangentMode=RCTM_User, time=0, value=1}, {interpMode=RCIM_Cubic, tangentMode=RCTM_User, time=1, value=0}]}}})
    ScaleA=true
    ScaleRGB=true
   )
  ]
  rendererClasses [&NiagaraRibbonRendererProperties]
  renderers [{rendererClass=&NiagaraRibbonRendererProperties}]
 }
 NE_VC_Lightning_02 {
  emitterUpdateScript [
   &EmitterState(Life Cycle Mode=::NewEnumerator1, Loop Duration=1, UseLoopDelay=false)
   &VC_FilteredSocketCount()
   &VC_SpawnRibbon_v1(Number of Ribbons=$Emitter.PortCount, SpawnRate=50)
  ]
  particleSpawnScript [
   &VC_PaintedVariantsOverride() {enabled=false}
   &VC_InitializeRibbon_v1(
    Enable Skeletal Position=true
    Lifetime=&MakeCustomFloatFromBool(Bool=$User.bIsFrontendPreview, True Float=0.8, False Float=1)
    Offset Amount:NiagaraFloat=&MakeCustomFloatFromBool(Bool=$User.bIsFrontendPreview, True Float=40, False Float=20)
    Offset Axis:ENiagaraOrientationAxis=::XAxis
    Ribbon Width=20
   )
   &AddVelocity(Velocity=&MakeCustomVectorFromBool(Bool=$User.bIsFrontendPreview, True Vector=<-2000 0 1>, False Vector=<-1000 0 0>)) {module=AddVelocity001}
   &Color(Color=&RandomRangeLinearColor(Minimum=<rgba -3 -1 -3 1>, Maximum=<rgba 3 1 3 1>))
  ]
  particleUpdateScript [
   &ParticleState()
   &InheritSourceMovement(
    Applied Position Scale=&Multiply_VectorByFloat(
     Vector=<1 1 1>
     Float=&FloatFromCurve(FloatCurve=DataInterface() {propertyValues {Curve {keys [
          {
           arriveTangentWeight=0.4252
           interpMode=RCIM_Cubic
           leaveTangentWeight=0.4252
           tangentMode=RCTM_User
           tangentWeightMode=RCTWM_WeightedBoth
           time=0.05
           value=1
          }
          {interpMode=RCIM_Cubic, tangentMode=RCTM_User, time=1, value=0}
         ]}}})
    )
    ApplyForce=false
    ApplyPosition=true
    ApplyPositionFalloff=false
   )
   &SolveForcesAndVelocity(Clamp Velocity=false, Limit Acceleration=false)
   &VC_UpdateRibbon_v1(Enable Skeletal Position=true)
   &DynamicMaterialParameters(
    Index 0 Param 1=0.5
    Index 0 Param 2=20
    Module.Index0Param0Write=true
    Module.Index0Param1Write=true
    Module.Index0Param2Write=true
    Module.Index0Param3Write=true
   )
   &ScaleColor(
    Scale Alpha=&FloatFromCurve(FloatCurve=DataInterface() {propertyValues {Curve {keys [{interpMode=RCIM_Cubic, tangentMode=RCTM_User, time=0, value=1}, {interpMode=RCIM_Cubic, tangentMode=RCTM_User, time=1, value=0}]}}})
    Scale RGB=&VectorFromCurve(VectorCurve=DataInterface() {propertyValues {
       XCurve {keys [<at 0 1 RCIM_Cubic>,<at 0.3 -2 RCIM_Cubic>]}
       YCurve {keys [<at 0 1 RCIM_Cubic>,<at 0.3 -1 RCIM_Cubic>]}
       ZCurve {keys [<at 0 1 RCIM_Cubic>,<at 0.3 -1.5 RCIM_Cubic>]}
      }})
    ScaleA=true
    ScaleRGB=true
   )
  ]
  rendererClasses [&NiagaraRibbonRendererProperties]
  renderers [{rendererClass=&NiagaraRibbonRendererProperties}]
 }
}
summary {
 systemName=NS_Boost_Powershot
 emitters [
  {emitterName=NE_VC_CoreTrail, rendererClasses [&NiagaraRibbonRendererProperties]}
  {emitterName=NE_VC_CoreLightningTrail001, rendererClasses [&NiagaraRibbonRendererProperties]}
  {emitterName=NE_VC_Lightning_01, rendererClasses [&NiagaraRibbonRendererProperties]}
  {emitterName=NE_VC_Lightning_02, rendererClasses [&NiagaraRibbonRendererProperties]}
  {emitterName=NE_VC_Embers, rendererClasses [&NiagaraSpriteRendererProperties]}
  {emitterName=NE_VC_Base_Sparks_01, rendererClasses [&NiagaraSpriteRendererProperties]}
  {emitterName=NE_VC_Base_Sparks_01001, rendererClasses [&NiagaraSpriteRendererProperties]}
 ]
 userVariables [
  User.bIsFrontendPreview {type=NiagaraBool}
  User.bIsPainted {type=NiagaraBool}
  User.ColorVariantID {type=NiagaraInt32}
  User.DTCarOffset {type=Vector3f, description="", defaultValue {struct=Vector3f, value=<0 0 0>}}
  User.IsPainted {type=NiagaraBool}
  User.Offset {type=NiagaraPosition, description="", defaultValue {}}
  User.PaintedColor {type=LinearColor, description="", defaultValue {struct=LinearColor, value=<rgba 1 1 1 1>}}
  User.PaintedScalar {type=NiagaraFloat, defaultValue=1, description=""}
  User.PortCount {type=NiagaraInt32, defaultValue=1}
  User.Rescale {type=NiagaraFloat, defaultValue=2.5, description=""}
  User.VariableForPainted_8 {type=LinearColor, description="", defaultValue {struct=LinearColor, value=<rgba 1 1 1 1>}}
 ]
}
