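# Niagara system for the Powershot vehicle boost cosmetic, pairing two ribbon trails with swirling spark sprites and small drifting embers.
# Ribbons and sprites spawn per boost socket on the vehicle's skeletal mesh, with source-movement inheritance fading over life on trails and embers.
# Painted-variant overrides recolor sparks, embers and the core trail, while the thinner ribbon uses a user trail color when painted.
# That ribbon also computes a sine-wave offset and time-flickered channel values.
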NiagaraSystem /VehicleCosmetics/Boosts/Boost_Powershot/FX/NS_Boost_Powershot_Opt
dependencies {
 usedDataInterfaces [
  &NiagaraDataInterfaceCurve
  &NiagaraDataInterfaceArrayColor
  &NiagaraDataInterfaceVectorCurve
  &NiagaraDataInterfaceSkeletalMesh
 ]
 usedDynamicInputs []
 usedModules [
  &self:SystemSpawnScript.SystemScriptSource.SystemScriptGraph.NiagaraNodeAssignment_0.SetVariables_#1
  &SystemState
 ]
}
emitters {
 NE_VC_Base_Sparks_01002 {
  emitterUpdateScript [
   &EmitterState(Life Cycle Mode=::NewEnumerator1, Loop Duration=1, UseLoopDelay=false)
   &VC_FilteredSocketCount()
   &SpawnRate(SpawnRate=80, Use Spawn Probability=false)
  ]
  particleSpawnScript [
   &VC_PaintedVariantsOverride(
    Black=&SelectLinearColorFromArray(
     Array Sampling Mode=::NewEnumerator0
     Direct Array Index=?
     Color Selection Array=DataInterface() {propertyValues {ColorData [<rgba 1 1 1 10>, <rgba 13.3333 0 80 10>]}}
     Random Seed=0
    )
    BurntSienna=&SelectLinearColorFromArray(
     Array Sampling Mode=::NewEnumerator0
     Direct Array Index=?
     Color Selection Array=DataInterface() {propertyValues {ColorData [<rgba 5 0.8333 0 10>, <rgba 13.3333 0 80 10>]}}
     Random Seed=0
    )
    Cobalt=&SelectLinearColorFromArray(
     Array Sampling Mode=::NewEnumerator0
     Direct Array Index=?
     Color Selection Array=DataInterface() {propertyValues {ColorData [<rgba 0 13.9496 80 10>, <rgba 13.3333 0 80 10>]}}
     Random Seed=0
    )
    Crimson=&SelectLinearColorFromArray(
     Array Sampling Mode=::NewEnumerator0
     Direct Array Index=?
     Color Selection Array=DataInterface() {propertyValues {ColorData [<rgba 1 0 0 10>, <rgba 16.6667 0 100 10>]}}
     Random Seed=0
    )
    DefaultColor=&SelectLinearColorFromArray(
     Array Sampling Mode=::NewEnumerator0
     Direct Array Index=?
     Color Selection Array=DataInterface() {propertyValues {ColorData [<rgba 100 6.6667 0 10>, <rgba 13.3333 0 80 10>]}}
     Random Seed=0
    )
    ForestGreen=&SelectLinearColorFromArray(
     Array Sampling Mode=::NewEnumerator0
     Direct Array Index=?
     Color Selection Array=DataInterface() {propertyValues {ColorData [<rgba 0 20 1.0105 1>, <rgba 13.3333 0 80 10>]}}
     Random Seed=0
    )
    Grey=&SelectLinearColorFromArray(
     Array Sampling Mode=::NewEnumerator0
     Direct Array Index=?
     Color Selection Array=DataInterface() {propertyValues {ColorData [<rgba 50 50 50 10>, <rgba 13.3333 0 80 10>]}}
     Random Seed=0
    )
    Lime=&SelectLinearColorFromArray(
     Array Sampling Mode=::NewEnumerator0
     Direct Array Index=?
     Color Selection Array=DataInterface() {propertyValues {ColorData [<rgba 13.3333 0 80 10>, <rgba 0 20 1.0105 1>]}}
     Random Seed=0
    )
    Orange=&SelectLinearColorFromArray(
     Array Sampling Mode=::NewEnumerator0
     Direct Array Index=?
     Color Selection Array=DataInterface() {propertyValues {ColorData [<rgba 100 6.6667 0 10>, <rgba 13.3333 0 80 10>]}}
     Random Seed=0
    )
    Pink=&SelectLinearColorFromArray(
     Array Sampling Mode=::NewEnumerator0
     Direct Array Index=?
     Color Selection Array=DataInterface() {propertyValues {ColorData [<rgba 46.6667 30 50 10>, <rgba 0.5 0 3 10>]}}
     Random Seed=0
    )
    Purple=&SelectLinearColorFromArray(
     Array Sampling Mode=::NewEnumerator0
     Direct Array Index=?
     Color Selection Array=DataInterface() {propertyValues {ColorData [<rgba 80 60 100 10>, <rgba 13.3333 0 80 10>]}}
     Random Seed=0
    )
    R1 Black:LinearColor=&SelectLinearColorFromArray(
     Array Sampling Mode=::NewEnumerator0
     Direct Array Index=?
     Color Selection Array=DataInterface() {propertyValues {ColorData [<rgba 1 1 1 0>, <rgba 0.5 0 3 1>]}}
     Random Seed=0
    )
    R1 BurntSienna:LinearColor=&SelectLinearColorFromArray(
     Array Sampling Mode=::NewEnumerator0
     Direct Array Index=?
     Color Selection Array=DataInterface() {propertyValues {ColorData [<rgba 1 0.1667 0 1>, <rgba 0.8333 0 5 1>]}}
     Random Seed=0
    )
    R1 Cobalt:LinearColor=&SelectLinearColorFromArray(
     Array Sampling Mode=::NewEnumerator0
     Direct Array Index=?
     Color Selection Array=DataInterface() {propertyValues {ColorData [<rgba 0 0.1744 1 1>, <rgba 0.1667 0 1 2>]}}
     Random Seed=0
    )
    R1 Crimson:LinearColor=&SelectLinearColorFromArray(
     Array Sampling Mode=::NewEnumerator0
     Direct Array Index=?
     Color Selection Array=DataInterface() {propertyValues {ColorData [<rgba 2 0 0 0.2>, <rgba 0.1667 0 1 2>]}}
     Random Seed=0
    )
    R1 DefaultColor=&SelectLinearColorFromArray(
     Array Sampling Mode=::NewEnumerator0
     Direct Array Index=?
     Color Selection Array=DataInterface() {propertyValues {ColorData [<rgba 1 0.0667 0 1>, <rgba 0.1667 0 1 2.2>]}}
     Random Seed=0
    )
    R1 ForestGreen:LinearColor=&SelectLinearColorFromArray(
     Array Sampling Mode=::NewEnumerator0
     Direct Array Index=?
     Color Selection Array=DataInterface() {propertyValues {ColorData [<rgba 0.2 1 0.2404 0.8>, <rgba 0.3333 0 2 1>]}}
     Random Seed=0
    )
    R1 Grey:LinearColor=&SelectLinearColorFromArray(
     Array Sampling Mode=::NewEnumerator0
     Direct Array Index=?
     Color Selection Array=DataInterface() {propertyValues {ColorData [<rgba 1 1 1 1>, <rgba 0.3333 0 2 1>]}}
     Random Seed=0
    )
    R1 Lime:LinearColor=&SelectLinearColorFromArray(
     Array Sampling Mode=::NewEnumerator0
     Direct Array Index=?
     Color Selection Array=DataInterface() {propertyValues {ColorData [<rgba 0.1667 0 1 2>, <rgba 0.7226 1 0.0169 1>]}}
     Random Seed=0
    )
    R1 Orange:LinearColor=&SelectLinearColorFromArray(
     Array Sampling Mode=::NewEnumerator0
     Direct Array Index=?
     Color Selection Array=DataInterface() {propertyValues {ColorData [<rgba 1 0.4725 0 1>, <rgba 0.3333 0 2 1>]}}
     Random Seed=0
    )
    R1 Pink:LinearColor=&SelectLinearColorFromArray(
     Array Sampling Mode=::NewEnumerator0
     Direct Array Index=?
     Color Selection Array=DataInterface() {propertyValues {ColorData [<rgba 0.95 0.7 1 1>, <rgba 0.1667 0 1 2>]}}
     Random Seed=0
    )
    R1 Purple:LinearColor=&SelectLinearColorFromArray(
     Array Sampling Mode=::NewEnumerator0
     Direct Array Index=?
     Color Selection Array=DataInterface() {propertyValues {ColorData [<rgba 0.8 0.6 1 0>, <rgba 0.5 0 3 2>]}}
     Random Seed=0
    )
    R1 Saffron:LinearColor=&SelectLinearColorFromArray(
     Array Sampling Mode=::NewEnumerator0
     Direct Array Index=?
     Color Selection Array=DataInterface() {propertyValues {ColorData [<rgba 1 0.8333 0 1>, <rgba 0.1667 0 1 2>]}}
     Random Seed=0
    )
    R1 SkyBlue:LinearColor=&SelectLinearColorFromArray(
     Array Sampling Mode=::NewEnumerator0
     Direct Array Index=?
     Color Selection Array=DataInterface() {propertyValues {ColorData [<rgba 0 1 0.6667 1.2>, <rgba 0.1667 0 1 2>]}}
     Random Seed=0
    )
    R1 TitaniumWhite:LinearColor=&SelectLinearColorFromArray(
     Array Sampling Mode=::NewEnumerator0
     Direct Array Index=?
     Color Selection Array=DataInterface() {propertyValues {ColorData [<rgba 1 1 1 1>, <rgba 0.1667 0 1 2>]}}
     Random Seed=0
    )
    Saffron=&SelectLinearColorFromArray(
     Array Sampling Mode=::NewEnumerator0
     Direct Array Index=?
     Color Selection Array=DataInterface() {propertyValues {ColorData [<rgba 100 83.3333 0 10>, <rgba 13.3333 0 80 10>]}}
     Random Seed=0
    )
    SkyBlue=&SelectLinearColorFromArray(
     Array Sampling Mode=::NewEnumerator0
     Direct Array Index=?
     Color Selection Array=DataInterface() {propertyValues {ColorData [<rgba 0 100 66.6667 10>, <rgba 13.3333 0 80 10>]}}
     Random Seed=0
    )
    TitaniumWhite=&SelectLinearColorFromArray(
     Array Sampling Mode=::NewEnumerator0
     Direct Array Index=?
     Color Selection Array=DataInterface() {propertyValues {ColorData [<rgba 100 100 100 10>, <rgba 13.3333 0 80 10>]}}
     Random Seed=0
    )
    Use R1 Painted Overrides=true
   )
   &InitializeParticle(
    Color Mode=::NewEnumerator0
    Lifetime=&RandomRangeFloat(Minimum=0.5, Maximum=0.7)
    Sprite Rotation Mode=::NewEnumerator2
    Sprite Size Max=<1.5 6>
    Sprite Size Min=<1 4>
    Sprite Size Mode=::NewEnumerator2
    UsePositionOffset=false
    Write Material Random=true
   )
   &PartitionParticles(Number Of Partitions=$Emitter.PortCount)
   &SkeletalMeshLocation(
    Bone / Socket Index=$Particles.PartitionParticles.Partition
    Bone / Socket Sampling Mode=::NewEnumerator5
    MaskSpawnGroup=false
    Mesh Sampling Type=::NewEnumerator6
    Sampled Position Offset=<5 0 0>
    Skeletal Mesh=$System.SkeletalMesh
   )
   &AddVelocity(Velocity=&RandomRangeVector(Minimum=<-150 -700 -500>, Maximum=<-300 700 500>))
   /SetVariables(
    Particles.BlueColor:LinearColor=&MultiplyLinearColor(
     Linear Color=$Particles.R1 Color
     Scale Factor=&MakeVector4FromVectorAndFloat("Vector (RGB)"=<1 1 1>, "Float (Alpha)"=5)
    )
    Particles.Fakecolor:Vector2f=&RandomRangeVector2D(Minimum=<8 20>, Maximum=<10 25>)
   ) {isSetParameters=true}
  ]
  particleUpdateScript [
   &ParticleState()
   &SpriteRotationRate(Rotation Rate=&RandomRangeFloat(Minimum=-10, Maximum=10))
   &AccelerationForce(Acceleration=&Multiply_VectorByFloat(
     Vector=&RandomRangeVector(Minimum=<0 -200 -150>, Maximum=<0 200 150>)
     Float=&Multiply_Float(
      A=&FloatFromCurve(
       FloatCurve=DataInterface() {propertyValues {Curve {keys [<at 0 0 RCIM_Cubic>, {interpMode=RCIM_Cubic, tangentMode=RCTM_User, time=0.25, value=2}, {tangentMode=RCTM_User, time=0.75, value=-2}]}}}
       Scale Curve=4
      )
      B=1
     )
    )) {module=AccelerationForce001}
   &CurlNoiseForce(Noise Strength=200, Pan Noise=true, Pan Noise Field=<1 0 0>)
   &VortexForce(
    Origin Pull Amount:NiagaraFloat=&RandomRangeFloat(Minimum=5000, Maximum=8000, Evaluation Type=::NewEnumerator1)
    UseOriginOffset=true
    Vortex Axis=&RandomRangeVector(Minimum=<1 0 0>, Maximum=<-1 0 0>, Evaluation Type=::NewEnumerator1)
    Vortex Axis Coordinate Space:ENiagaraCoordinateSpace=::Local
    Vortex Force Amount=&RandomRangeFloat(Minimum=3000, Maximum=5000)
    Vortex Origin:NiagaraPosition=$Particles.Initial.Position
    Vortex Origin Offset:Vector3f=&RandomRangeVector(Minimum=<0 0 0>, Maximum=<-20 0 0>, Evaluation Type=::NewEnumerator1)
    Vortex Origin Offset Coordinate Space:ENiagaraCoordinateSpace=::Local
   )
   &SolveForcesAndVelocity(Clamp Velocity=false, Limit Acceleration=false)
   &ScaleSpriteSize(Uniform Curve Sprite Scale=DataInterface() {propertyValues {Curve {keys [
        {arriveTangent=0.3218, interpMode=RCIM_Cubic, leaveTangent=0.3218, tangentMode=RCTM_User, time=0.25, value=1}
        <at 0.6 2 RCIM_Cubic>
        <at 1 1 RCIM_Cubic>
       ]}}})
  ]
  rendererClasses [&NiagaraSpriteRendererProperties]
  renderers [
   {rendererClass=&NiagaraSpriteRendererProperties}
   {rendererClass=&NiagaraSpriteRendererProperties, rendererIndex=1}
  ]
 }
 NE_VC_Boost_Ribbon_Parent {
  emitterUpdateScript [
   &EmitterState(Life Cycle Mode=::NewEnumerator1, Loop Duration=1, UseLoopDelay=false)
   &VC_FilteredSocketCount()
   &VC_RateQualityHandler()
   /SetVariables(
    Emitter.TimeLerp:NiagaraFloat=&Power(a=HlslExpression() {hlslExpression="(fmod(Engine.Time * 0.325, 0.4) + 0.15) * 3"}, b=2)
    Emitter.TimeOffset:NiagaraFloat=&Power(a=HlslExpression() {hlslExpression="fmod(Engine.Time, 0.3) + 0.15 * 3"}, b=4)
   ) {isSetParameters=true}
   /SetVariables(
    Emitter.LerpChannel:Vector3f=&Lerp_Vector(
     A=<0 0 0>
     B=<0 1 0>
     Alpha=&Ceil_Float(Float=&Sine(Normalized Angle=$Emitter.Age, Period=$Emitter.TimeLerp, Scale=1, Bias=0))
    )
    Emitter.OffsetX:NiagaraFloat=&Multiply_Float(
     A=&Ceil_Float(Float=&Sine(Normalized Angle=$Emitter.Age, Period=$Emitter.TimeOffset, Scale=1, Bias=0))
     B=0.25
    )
   ) {isSetParameters=true}
   &VC_SpawnRibbon_v1(Number of Ribbons=$Emitter.PortCount, SpawnRate=60)
  ]
  particleSpawnScript [
   &VC_PaintedVariantsOverride() {enabled=false}
   /SetVariables(Particles.ColorTwo:LinearColor=<rgba 1.0624 2.6209 10 1>, Particles.SizeTwo=80) {isSetParameters=true}
   &VC_InitializeRibbon_v1(
    Color=&MakeCustomLinearColorFromBool(Bool=$User.bIsPainted, True Color=$User.BelowTrailColor, False Color=<rgba 3.369 2.7826 40 1>)
    Custom Offset=true
    Enable Skeletal Position=true
    Lifetime=0.6
    Local Vector Offset=&AddVectorToPosition(Position=<-50 0 -10>, Vector=$User.DTCarOffset)
    Offset Amount=1
    Ribbon Width=10
   )
   &AddVelocity(Velocity=<-2000 0 50>) {module=AddVelocity001}
   &ShapeLocation(MaskSpawnGroup=false, Override Seed=false, Sphere Radius=7, UseOwnerScale=true)
  ]
  particleUpdateScript [
   &ParticleState()
   &ScaleColor(
    Scale Alpha=&MakeCustomFloatFromBool(
     Bool=$User.bIsPainted
     True Float=0.15
     False Float=&FloatFromCurve(
      FloatCurve=DataInterface() {propertyValues {Curve {
         defaultValue=0
         keys [{interpMode=RCIM_Cubic, tangentMode=RCTM_User, time=0.6115, value=1}, {interpMode=RCIM_Cubic, tangentMode=RCTM_User, time=0.85, value=0}]
        }}}
      Scale Curve=0.2
     )
    )
    ScaleA=true
    ScaleRGB=true
   )
   &ScaleRibbonWidth(Ribbon Width Scale=&FloatFromCurve(
     FloatCurve=DataInterface() {propertyValues {Curve {keys [
         {interpMode=RCIM_Cubic, tangentMode=RCTM_User, time=0, value=0}
         {arriveTangent=0.5028, interpMode=RCIM_Cubic, leaveTangent=0.5028, tangentMode=RCTM_User, time=0.1, value=1.25}
         {arriveTangent=-0.5446, interpMode=RCIM_Cubic, leaveTangent=-0.5446, tangentMode=RCTM_User, time=0.8, value=1}
         {arriveTangent=-3.985, interpMode=RCIM_Cubic, leaveTangent=-3.985, tangentMode=RCTM_User, time=0.9, value=0}
        ]}}}
     Scale Curve=1.25
    ))
   /SetVariables(Particles.SizeScaleTwo:NiagaraFloat=&Multiply_Float(
     A=$Particles.SizeTwo
     B=&FloatFromCurve(FloatCurve=DataInterface() {propertyValues {Curve {
         defaultValue=1
         keys [{interpMode=RCIM_Cubic, tangentMode=RCTM_User, time=0, value=1}, {interpMode=RCIM_Cubic, tangentMode=RCTM_User, time=1, value=0}]
        }}})
    )) {isSetParameters=true}
   &InheritSourceMovement(
    Applied Position Scale=&Multiply_VectorByFloat(
     Vector=<1 1 1>
     Float=&FloatFromCurve(FloatCurve=DataInterface() {propertyValues {Curve {keys [
          {
           arriveTangentWeight=0.4252
           interpMode=RCIM_Cubic
           leaveTangentWeight=0.4252
           tangentMode=RCTM_User
           tangentWeightMode=RCTWM_WeightedBoth
           time=0.05
           value=1
          }
          {interpMode=RCIM_Cubic, tangentMode=RCTM_User, time=1, value=0}
         ]}}})
    )
    ApplyForce=false
    ApplyPosition=true
    ApplyPositionFalloff=false
   )
   &SolveForcesAndVelocity(Clamp Velocity=false, Limit Acceleration=false)
   &VC_UpdateRibbon_v1(Enable Skeletal Position=true)
   /SetVariables(Particles.WPO:NiagaraPosition=&AddVectorToPosition(
     Position=$Particles.Position
     Vector=&TransformVector(
      Vector=&MakeVector(
       X=0
       Y=0
       Z=&Sine(
        Normalized Angle=&Add_Float(
         A=&Multiply_Float(A=$Particles.NormalizedAge, B=-1)
         B=&Multiply_Float(A=$Emitter.LoopedAge, B=1.1)
        )
        Period=1
        Scale=&FloatFromCurve(
         FloatCurve=DataInterface() {propertyValues {Curve {
            defaultValue=0
            keys [
             {arriveTangent=9.0725, interpMode=RCIM_Cubic, leaveTangent=9.0725, tangentMode=RCTM_User, time=0, value=0}
             {arriveTangent=-2.5249, interpMode=RCIM_Cubic, leaveTangent=-2.5249, tangentMode=RCTM_User, time=0.1408, value=0.9847}
             {arriveTangent=-0.0405, interpMode=RCIM_Cubic, leaveTangent=-0.0405, tangentMode=RCTM_User, time=0.7, value=0}
            ]
           }}}
         Scale Curve=15
        )
        Bias=1
       )
      )
      Destination Space=::Simulation
      Apply Scale=true
     )
    )) {isSetParameters=true}
   /SetVariables(
    Particles.Fakeglow:LinearColor=&MultiplyLinearColor(
     Linear Color=$Particles.Color
     Scale Factor=&MakeVector4FromVectorAndFloat(
      "Vector (RGB)"=<1 1 1>
      "Float (Alpha)"=&FloatFromCurve(FloatCurve=DataInterface() {propertyValues {Curve {
          defaultValue=1
          keys [
           {arriveTangent=-184.2221, interpMode=RCIM_Cubic, leaveTangent=-184.2221, tangentMode=RCTM_User, time=0, value=50}
           {interpMode=RCIM_Cubic, tangentMode=RCTM_User, time=0.4, value=0}
          ]
         }}})
     )
    )
    Particles.MainLightning:LinearColor=&MultiplyLinearColor(
     Linear Color=<rgba 1.0624 2.6209 10 1>
     Scale Factor=&MakeVector4FromVectorAndFloat(
      "Vector (RGB)"=<1 1 1>
      "Float (Alpha)"=&FloatFromCurve(FloatCurve=DataInterface() {propertyValues {Curve {
          defaultValue=1
          keys [{interpMode=RCIM_Cubic, tangentMode=RCTM_User, time=0.5, value=1}, {interpMode=RCIM_Cubic, tangentMode=RCTM_User, time=1, value=0}]
         }}})
     )
    )
   ) {isSetParameters=true}
  ]
  rendererClasses [&NiagaraRibbonRendererProperties]
  renderers [
   {rendererClass=&NiagaraRibbonRendererProperties}
   {rendererClass=&NiagaraRibbonRendererProperties, rendererIndex=1}
  ]
 }
 NE_VC_CoreTrail002 {
  emitterUpdateScript [
   &EmitterState(Inactive Response=::NewEnumerator1, Life Cycle Mode=::NewEnumerator1, Loop Duration=1, UseLoopDelay=false)
   &VC_FilteredSocketCount()
   &VC_RateQualityHandler(High Spawn Rate=60, Low Spawn Rate=60, Medium Spawn Rate=60, Use Custom Rate?=true)
   &VC_SpawnRibbon_v1(Number of Ribbons=$Emitter.PortCount, SpawnRate=$Emitter.Spawn Rate)
  ]
  particleSpawnScript [
   &VC_PaintedVariantsOverride(
    Black=<rgba 0 0 0 5>
    BurntSienna=<rgba 0.6 0.1853 0.074 5>
    Cobalt=<rgba 0 0.0185 1 5>
    Crimson=<rgba 0.6 0 0.0163 5>
    DefaultColor=<rgba 0.25 0 1.5 1.25>
    ForestGreen=<rgba 0 0.2 0.0101 5>
    Grey=<rgba 0.1 0.1 0.1 5>
    Lime=<rgba 0.3 0.15 0 5>
    Orange=<rgba 0.6 0.2093 0 5>
    Pink=<rgba 0.575 0.3 0.6 5>
    Purple=<rgba 0.2873 0 0.7 5>
    R1 Black:LinearColor=<rgba 0.06 0.06 0.06 5>
    R1 BurntSienna:LinearColor=<rgba 0.5 0.1544 0.0616 1>
    R1 DefaultColor=<rgba 5 1.8275 0.95 3>
    R1 ForestGreen:LinearColor=<rgba 0 0.3 0 1>
    R1 Grey:LinearColor=<rgba 0.4 0.4 0.4 1>
    R1 Lime:LinearColor=<rgba 0.3333 0.4 0 1>
    R1 Orange:LinearColor=<rgba 0.7 0.4083 0 1>
    R1 Pink:LinearColor=<rgba 0.8 0.4 0.6545 1>
    R1 Purple:LinearColor=<rgba 0.1472 0.0014 0.15 2>
    R1 Saffron:LinearColor=<rgba 0.5 0.3833 0 2>
    R1 SkyBlue:LinearColor=<rgba 0 0.3992 0.6 2>
    R1 TitaniumWhite:LinearColor=<rgba 0.7 0.7 0.7 1>
    Saffron=<rgba 0.6 0.3 0 5>
    SkyBlue=<rgba 0 0 0.8 5>
    TitaniumWhite=<rgba 0.4 0 0.6 5>
    Use R1 Painted Overrides=true
   )
   &VC_InitializeRibbon_v1(
    Color=$Particles.Color
    Custom Offset=true
    Enable Skeletal Position=true
    Lifetime=0.6
    Local Vector Offset=&AddVectorToPosition(Position=<115 0 0>, Vector=$User.DTCarOffset)
    Offset Amount=1
    Ribbon Width=130
   )
   &AddVelocity(Velocity=<-1500 0 0>) {module=AddVelocity001}
  ]
  particleUpdateScript [
   &ParticleState()
   &SolveForcesAndVelocity(Clamp Velocity=false, Limit Acceleration=false)
   &InheritSourceMovement(
    Applied Position Scale=&Multiply_VectorByFloat(
     Vector=<1 1 1>
     Float=&FloatFromCurve(FloatCurve=DataInterface() {propertyValues {Curve {keys [
          {
           arriveTangentWeight=0.4252
           interpMode=RCIM_Cubic
           leaveTangentWeight=0.4252
           tangentMode=RCTM_User
           tangentWeightMode=RCTWM_WeightedBoth
           time=0.05
           value=1
          }
          {interpMode=RCIM_Cubic, tangentMode=RCTM_User, time=1, value=0}
         ]}}})
    )
    ApplyForce=false
    ApplyPosition=true
    ApplyPositionFalloff=false
   )
   &SolveForcesAndVelocity(Clamp Velocity=false, Limit Acceleration=true) {module=SolveForcesAndVelocity001}
   &VC_UpdateRibbon_v1(Enable Skeletal Position=true)
   /SetVariables(
    Particles.RedColor:LinearColor=&MultiplyLinearColor(
     Linear Color=$Particles.Initial.R1 Color
     Scale Factor=&MakeVector4FromVectorAndFloat(
      "Vector (RGB)"=&MakeCustomVectorFromBool(
       Bool=$User.bIsPainted
       True Vector=&VectorFromFloat(Value=10)
       False Vector=&VectorFromCurve(VectorCurve=DataInterface() {propertyValues {
          XCurve {keys [<at 0 2>,<at 0.5 1>,<at 0.7 1>]}
          YCurve {keys [<at 0 0.3>,<at 0.3 1>,<at 0.4 1>,<at 0.45 0>]}
          ZCurve {keys [<at 0 0.2>,<at 0.3 1>,<at 0.4 1>,<at 0.45 0>]}
         }})
      )
      "Float (Alpha)"=&FloatFromCurve(
       FloatCurve=DataInterface() {propertyValues {Curve {keys [<at 0 1 RCIM_Cubic>,<at 0.8 0 RCIM_Cubic>]}}}
       Scale Curve=&MakeCustomFloatFromBool(Bool=$User.bIsPainted, True Float=2, False Float=1)
      )
     )
    )
    Particles.RedWidth:NiagaraFloat=&FloatFromCurve(
     FloatCurve=DataInterface() {propertyValues {Curve {
        defaultValue=1
        keys [
         {interpMode=RCIM_Cubic, tangentMode=RCTM_User, time=0, value=1.5}
         {arriveTangent=-1.25, interpMode=RCIM_Cubic, leaveTangent=-1.25, time=0.2985, value=0.7355}
         {interpMode=RCIM_Cubic, tangentMode=RCTM_User, time=1, value=0.25}
        ]
       }}}
     Scale Curve=65
    )
   ) {isSetParameters=true}
   &ScaleRibbonWidth(Ribbon Width Scale=&FloatFromCurve(
     FloatCurve=DataInterface() {propertyValues {Curve {
        defaultValue=1
        keys [
         {arriveTangent=-1.5968, interpMode=RCIM_Cubic, leaveTangent=-1.5968, tangentMode=RCTM_User, time=0, value=1.5}
         {arriveTangent=-0.1769, interpMode=RCIM_Cubic, leaveTangent=-0.1769, tangentMode=RCTM_User, time=0.55, value=0.7}
         {arriveTangentWeight=0.1253, interpMode=RCIM_Cubic, tangentMode=RCTM_User, time=0.95, value=0}
        ]
       }}}
     Scale Curve=1.15
    ))
   &ScaleColor(
    Scale Alpha=&FloatFromCurve(
     FloatCurve=DataInterface() {propertyValues {Curve {
        defaultValue=1
        keys [
         <at 0.15 0.2 RCIM_Cubic>
         {arriveTangent=1.9209, interpMode=RCIM_Cubic, leaveTangent=1.9209, time=0.3, value=1}
         {arriveTangent=-1.4286, interpMode=RCIM_Cubic, leaveTangent=-1.4286, time=0.5665, value=1}
         {arriveTangent=-0.1079, interpMode=RCIM_Cubic, leaveTangent=-0.1079, tangentMode=RCTM_User, time=1, value=0}
        ]
       }}}
     Scale Curve=10
    )
    Scale RGB=&Multiply_VectorByFloat(
     Vector=<1 1 1>
     Float=&FloatFromCurve(
      FloatCurve=DataInterface() {propertyValues {Curve {
         defaultValue=1
         keys [{interpMode=RCIM_Cubic, tangentMode=RCTM_User, time=0.5, value=1}, {interpMode=RCIM_Cubic, tangentMode=RCTM_User, time=1, value=0}]
        }}}
      Scale Curve=0.65
     )
    )
    ScaleA=true
    ScaleRGB=true
   ) {module=ScaleColor001}
  ]
  rendererClasses [&NiagaraRibbonRendererProperties]
  renderers [
   {rendererClass=&NiagaraRibbonRendererProperties}
   {rendererClass=&NiagaraRibbonRendererProperties, rendererIndex=1}
  ]
 }
 NE_VC_Embers001 {
  emitterUpdateScript [
   &EmitterState(Life Cycle Mode=::NewEnumerator1, Loop Duration=1, UseLoopDelay=false)
   &VC_FilteredSocketCount()
   &SpawnRate(SpawnRate=60, Use Spawn Probability=false)
  ]
  particleSpawnScript [
   &VC_PaintedVariantsOverride(
    Black=<rgba 1 1 1 10>
    BurntSienna=<rgba 100 6.6667 0 10>
    Cobalt=<rgba 0 13.9496 80 10>
    Crimson=<rgba 1 0 0 10>
    DefaultColor=<rgba 20 1 0.5 20>
    ForestGreen=<rgba 0 20 1.0105 1>
    Grey=<rgba 50 50 50 10>
    Lime=<rgba 0 20 1.0105 1>
    Orange=<rgba 100 6.6667 0 10>
    Pink=<rgba 83.3334 0 100 10>
    Purple=<rgba 80 60 100 10>
    Saffron=<rgba 100 83.3333 0 10>
    SkyBlue=<rgba 0 100 66.6667 10>
    TitaniumWhite=<rgba 100 100 100 10>
   )
   &InitializeParticle(
    Color Mode=::NewEnumerator0
    Lifetime=&RandomRangeFloat(Minimum=0.1, Maximum=0.5)
    Sprite Rotation Mode=::NewEnumerator2
    Sprite Size Max=<1.5 1.8>
    Sprite Size Min=<1.3 1.5>
    Sprite Size Mode=::NewEnumerator2
    UsePositionOffset=false
    Write Material Random=true
   )
   &PartitionParticles(Number Of Partitions=$Emitter.PortCount)
   &SkeletalMeshLocation(
    Bone / Socket Index=$Particles.PartitionParticles.Partition
    Bone / Socket Sampling Mode=::NewEnumerator5
    MaskSpawnGroup=false
    Mesh Sampling Type=::NewEnumerator6
    Skeletal Mesh=$System.SkeletalMesh
   )
   &AddVelocity(Velocity=&RandomRangeVector(Minimum=<-1200 -30 -30>, Maximum=<-1200 30 30>))
   &ShapeLocation(MaskSpawnGroup=false, Override Seed=false, Sphere Radius=15, Sphere Surface Distribution=1, UseOwnerScale=true)
  ]
  particleUpdateScript [
   &ParticleState()
   &SpriteRotationRate(Rotation Rate=&RandomRangeFloat(Minimum=-10, Maximum=10))
   &InheritSourceMovement(
    Applied Position Scale=&Multiply_VectorByFloat(
     Vector=<1 1 1>
     Float=&FloatFromCurve(FloatCurve=DataInterface() {propertyValues {Curve {keys [
          {
           arriveTangentWeight=0.4252
           interpMode=RCIM_Cubic
           leaveTangentWeight=0.4252
           tangentMode=RCTM_User
           tangentWeightMode=RCTWM_WeightedBoth
           time=0.05
           value=1
          }
          {interpMode=RCIM_Cubic, tangentMode=RCTM_User, time=1, value=0}
         ]}}})
    )
    ApplyForce=false
    ApplyPosition=true
    ApplyPositionFalloff=false
   )
   &CurlNoiseForce(Noise Strength=500, Pan Noise=true, Pan Noise Field=<100 0 0>)
   &Drag(Use Linear Drag=true, Use Rotational Drag=true)
   &SolveForcesAndVelocity(Clamp Velocity=false, Limit Acceleration=false)
   &ScaleSpriteSize(
    Uniform Curve Scale=2
    Uniform Curve Sprite Scale=DataInterface() {propertyValues {Curve {keys [
        {arriveTangent=0.3218, interpMode=RCIM_Cubic, leaveTangent=0.3218, tangentMode=RCTM_User, time=0.25, value=0}
        <at 0.75 1 RCIM_Cubic>
        <at 1 0 RCIM_Cubic>
       ]}}}
   )
  ]
  rendererClasses [&NiagaraSpriteRendererProperties]
  renderers [{rendererClass=&NiagaraSpriteRendererProperties}]
 }
}
summary {
 systemName=NS_Boost_Powershot_Opt
 emitters [
  {emitterName=NE_VC_Embers001, rendererClasses [&NiagaraSpriteRendererProperties]}
  {emitterName=NE_VC_Base_Sparks_01002, rendererClasses [&NiagaraSpriteRendererProperties]}
  {emitterName=NE_VC_Boost_Ribbon_Parent, rendererClasses [&NiagaraRibbonRendererProperties]}
  {emitterName=NE_VC_CoreTrail002, rendererClasses [&NiagaraRibbonRendererProperties]}
 ]
 userVariables [
  User.BelowTrailColor {type=LinearColor, description="", defaultValue {struct=LinearColor, value=<rgba 1 1 1 1>}}
  User.bIsFrontendPreview {type=NiagaraBool}
  User.bIsPainted {type=NiagaraBool}
  User.ColorVariantID {type=NiagaraInt32}
  User.DTCarOffset {type=Vector3f, description="", defaultValue {struct=Vector3f, value=<0 0 0>}}
  User.IsPainted {type=NiagaraBool}
  User.Offset {type=NiagaraPosition, description="", defaultValue {}}
  User.PaintedColor {type=LinearColor, description="", defaultValue {struct=LinearColor, value=<rgba 1 1 1 1>}}
  User.PaintedScalar {type=NiagaraFloat, defaultValue=1, description=""}
  User.PortCount {type=NiagaraInt32, defaultValue=1}
  User.Rescale {type=NiagaraFloat, defaultValue=2.5, description=""}
  User.VariableForPainted_8 {type=LinearColor, description="", defaultValue {struct=LinearColor, value=<rgba 1 1 1 1>}}
 ]
}
