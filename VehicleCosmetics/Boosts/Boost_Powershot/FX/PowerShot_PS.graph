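# Niagara system for the Powershot vehicle boost cosmetic: a continuous energy trail with lightning, glow, sparks and embers.
# Five ribbon emitters layer core, lightning and glow trails along local X, tapering in width and fading over life, with dynamic material parameters driving their look.
# Three sprite emitters add embers and red and blue-violet SubUV-animated sparks, the sparks swirled by vortex force.
# All values are set in the emitters, with no user parameters exposed.
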NiagaraSystem /VehicleCosmetics/Boosts/Boost_Powershot/FX/PowerShot_PS
dependencies {
 usedDataInterfaces [&NiagaraDataInterfaceVectorCurve, &NiagaraDataInterfaceCurve]
 usedDynamicInputs [&RandomRangeVector2D]
 usedModules [&SystemState]
}
emitters {
 BaseCoreTrail {
  emitterUpdateScript [
   &EmitterState(Life Cycle Mode=::NewEnumerator1, Loop Duration=4, UseLoopDelay=false)
   &SpawnRate(SpawnRate=60, Use Spawn Probability=false)
  ]
  particleSpawnScript [
   &InitializeParticle(
    Color=<rgba 1 0.1 1.8 5>
    Position Offset=<-210.4285 0 0>
    Ribbon Width=400
    Ribbon Width Mode=::NewEnumerator1
    UsePositionOffset=true
    Write Material Random=true
   )
   &AddVelocity(Velocity=<800 50 0>, Velocity Speed Scale=3)
   &AddVelocity(Velocity=&RandomRangeVector(Minimum=<0 -50 -50>, Maximum=<0 50 50>)) {enabled=false, module=AddVelocity001}
  ]
  particleUpdateScript [
   &DynamicMaterialParameters(
    Index 0 Param 1=3
    Index 0 Param 2=&RandomRangeFloat(Minimum=0, Maximum=1)
    Index 0 Param 4=0.5
    Module.Index0Param0Write=true
    Module.Index0Param1Write=true
    Module.Index0Param2Write=true
    Module.Index0Param3Write=true
   )
   &ParticleState()
   &ScaleVelocity(Velocity Scale=&RandomRangeVector(
     Minimum=&VectorFromCurve(VectorCurve=DataInterface() {propertyValues {XCurve {keys [<at 0 1>,<at 1 0>]}, YCurve {keys [<at 0 1>,<at 1 0>]}, ZCurve {keys [<at 0 1>,<at 1 0>]}}})
     Maximum=&VectorFromCurve(VectorCurve=DataInterface() {propertyValues {XCurve {keys [<at 0 1>,<at 1 0>]}, YCurve {keys [<at 0 1>,<at 1 0>]}, ZCurve {keys [<at 0 1>,<at 1 0>]}}})
    )) {enabled=false}
   &SolveForcesAndVelocity(Clamp Velocity=false, Limit Acceleration=true)
   &ScaleColor(
    Scale Alpha=&FloatFromCurve(FloatCurve=DataInterface() {propertyValues {Curve {keys [
         {interpMode=RCIM_Cubic, tangentMode=RCTM_Break, time=0, value=5}
         {interpMode=RCIM_Cubic, tangentMode=RCTM_Break, time=0.5528, value=1}
         {interpMode=RCIM_Cubic, tangentMode=RCTM_Break, time=0.9909, value=0}
        ]}}})
    Scale RGB=&VectorFromCurve(VectorCurve=DataInterface() {propertyValues {
       XCurve {keys [{interpMode=RCIM_Cubic, tangentMode=RCTM_Break, time=0, value=0.05}, <at 0.25 0.01>]}
       YCurve {keys [{interpMode=RCIM_Cubic, tangentMode=RCTM_Break, time=0, value=0.01}, <at 0.25 0.02>]}
       ZCurve {keys [{interpMode=RCIM_Cubic, tangentMode=RCTM_Break, time=0, value=0.1}, <at 0.25 0.025>]}
      }})
    ScaleA=true
    ScaleRGB=true
   )
   &ScaleRibbonWidth(Ribbon Width Scale=&FloatFromCurve(FloatCurve=DataInterface() {propertyValues {Curve {keys [<at 0 1>,<at 1 0>]}}}))
  ]
  rendererClasses [&NiagaraRibbonRendererProperties]
  renderers [{rendererClass=&NiagaraRibbonRendererProperties}]
 }
 BaseSparks01 {
  emitterUpdateScript [
   &EmitterState()
   &SpawnRate(Spawn Probability=2, SpawnRate=30, Use Spawn Probability=true)
   &SpawnBurst_Instantaneous(Spawn Count=15, Use Spawn Probability=true, UseLoopCountLimit=false) {enabled=false}
  ]
  particleSpawnScript [
   &InitializeParticle(
    Color=<rgba 2 0.1 0.05 5>
    Lifetime Max=1
    Lifetime Min=0.7
    Lifetime Mode=::NewEnumerator1
    Position Offset=<-200 0 -50>
    Sprite Rotation Angle=&RandomRangeFloat(Minimum=-30, Maximum=30)
    Sprite Rotation Mode=::NewEnumerator0
    Sprite Size Max=<40 20>
    Sprite Size Min=<20 10>
    Sprite Size Mode=::NewEnumerator2
    UsePositionOffset=true
    Write Material Random=true
   )
   &AddVelocity(Velocity=<900 0 0>, Velocity Speed Scale=4) {module=AddVelocity001}
   &AddVelocity(
    Rotation Coordinate Space=::World
    Velocity=&RandomRangeVector(Minimum=<0 -100 -300>, Maximum=<0 100 300>)
    Velocity Speed Scale=5
   )
   &ShapeLocation(Cone Length=70, MaskSpawnGroup=false, Non Uniform Scale=<1 0 1>, Override Seed=false, Shape Primitive=::NewEnumerator5, UseOwnerScale=true)
  ]
  particleUpdateScript [
   &ParticleState()
   &DynamicMaterialParameters(
    Index 0 Param 1=2
    Module.Index0Param0Write=true
    Module.Index0Param1Write=true
    Module.Index0Param2Write=true
    Module.Index0Param3Write=true
   )
   &VortexForce(UseOriginOffset=false, Vortex Force Amount=5)
   &ScaleVelocity(Velocity Scale=&VectorFromFloat(Value=&FloatFromCurve(FloatCurve=DataInterface() {propertyValues {Curve {keys [<at 0 0.12>,<at 0.1 0.1>]}}})))
   &Drag(Drag=0.5, Use Linear Drag=true, Use Rotational Drag=true) {enabled=false}
   &SolveForcesAndVelocity(Clamp Velocity=false, Limit Acceleration=false)
   &SpriteRotationRate(Rotation Rate=&RandomRangeFloat(Minimum=-10, Maximum=10)) {enabled=false}
   &ScaleColor(Scale RGB=<50 50 50>, ScaleA=true, ScaleRGB=true)
   &SubUVAnimation(
    End Frame Range Override=16
    "Sprite Renderer ":NiagaraDataInterfaceSpriteRendererInfo=Linked() {linkedVariable {name="Particles.Sprite Renderer ", type=NiagaraDataInterfaceSpriteRendererInfo}}
    UseEndFrame=true
    UseStartFrame=true
   )
   &ScaleSpriteSize(Uniform Curve Sprite Scale=DataInterface() {propertyValues {Curve {keys [{arriveTangent=0.3218, interpMode=RCIM_Cubic, leaveTangent=0.3218, tangentMode=RCTM_User, time=0, value=2}, <at 1 0.1 RCIM_Cubic>]}}})
   &SpriteFacingAndAlignment(WriteAlignment=false, WriteFacing=true) {enabled=false}
  ]
  rendererClasses [&NiagaraSpriteRendererProperties]
  renderers [{rendererClass=&NiagaraSpriteRendererProperties}]
 }
 BaseSparks01001 {
  emitterUpdateScript [
   &EmitterState()
   &SpawnRate(Spawn Probability=2, SpawnRate=30, Use Spawn Probability=true)
   &SpawnBurst_Instantaneous(Spawn Count=15, Use Spawn Probability=true, UseLoopCountLimit=false) {enabled=false}
  ]
  particleSpawnScript [
   &InitializeParticle(
    Color=<rgba 0.2 0.05 1 1>
    Lifetime Max=1
    Lifetime Min=0.7
    Lifetime Mode=::NewEnumerator1
    Position Offset=<-200 0 -50>
    Sprite Rotation Angle=&RandomRangeFloat(Minimum=-15, Maximum=15)
    Sprite Rotation Mode=::NewEnumerator0
    Sprite Size Max=<40 20>
    Sprite Size Min=<20 10>
    Sprite Size Mode=::NewEnumerator2
    UsePositionOffset=true
    Write Material Random=true
   )
   &AddVelocity(Velocity=<900 0 0>, Velocity Speed Scale=4) {module=AddVelocity001}
   &AddVelocity(
    Rotation Coordinate Space=::World
    Velocity=&RandomRangeVector(Minimum=<0 -100 -400>, Maximum=<0 100 400>)
    Velocity Speed Scale=5
   )
   &ShapeLocation(Cone Length=70, MaskSpawnGroup=false, Non Uniform Scale=<1 0 1>, Override Seed=false, Shape Primitive=::NewEnumerator5, UseOwnerScale=true)
  ]
  particleUpdateScript [
   &ParticleState()
   &DynamicMaterialParameters(
    Index 0 Param 1=2
    Module.Index0Param0Write=true
    Module.Index0Param1Write=true
    Module.Index0Param2Write=true
    Module.Index0Param3Write=true
   )
   &VortexForce(UseOriginOffset=false, Vortex Force Amount=5)
   &ScaleVelocity(Velocity Scale=&VectorFromFloat(Value=&FloatFromCurve(FloatCurve=DataInterface() {propertyValues {Curve {keys [<at 0 0.12>,<at 0.1 0.1>]}}})))
   &Drag(Drag=0.5, Use Linear Drag=true, Use Rotational Drag=true) {enabled=false}
   &SolveForcesAndVelocity(Clamp Velocity=false, Limit Acceleration=false)
   &SpriteRotationRate(Rotation Rate=&RandomRangeFloat(Minimum=-10, Maximum=10)) {enabled=false}
   &ScaleColor(Scale RGB=<50 50 50>, ScaleA=true, ScaleRGB=true)
   &SubUVAnimation(
    End Frame Range Override=16
    "Sprite Renderer ":NiagaraDataInterfaceSpriteRendererInfo=Linked() {linkedVariable {name="Particles.Sprite Renderer ", type=NiagaraDataInterfaceSpriteRendererInfo}}
    UseEndFrame=true
    UseStartFrame=true
   )
   &ScaleSpriteSize(Uniform Curve Sprite Scale=DataInterface() {propertyValues {Curve {keys [{arriveTangent=0.3218, interpMode=RCIM_Cubic, leaveTangent=0.3218, tangentMode=RCTM_User, time=0, value=2}, <at 1 0.1 RCIM_Cubic>]}}})
   &SpriteFacingAndAlignment(WriteAlignment=false, WriteFacing=true) {enabled=false}
  ]
  rendererClasses [&NiagaraSpriteRendererProperties]
  renderers [{rendererClass=&NiagaraSpriteRendererProperties}]
 }
 CoreGlow001 {
  emitterUpdateScript [
   &EmitterState(Life Cycle Mode=::NewEnumerator1, Loop Duration=4, UseLoopDelay=false)
   &SpawnRate(SpawnRate=15, Use Spawn Probability=false)
  ]
  particleSpawnScript [
   &InitializeParticle(
    Color=<rgba 2 4 7 50>
    Lifetime=0.65
    Position Offset=<-250 0 0>
    Ribbon Width=170
    Ribbon Width Mode=::NewEnumerator1
    UsePositionOffset=true
    Write Material Random=true
   )
   &AddVelocity(Velocity=<800 0 0>, Velocity Speed Scale=4)
   &AddVelocity(
    Velocity=&VectorFromCurve(VectorCurve=DataInterface() {propertyValues {XCurve {keys [<at 0 0>,<at 1 20>]}, YCurve {keys [<at 0 0>,<at 1 20>]}, ZCurve {keys [<at 0 0>,<at 1 20>]}}})
    Velocity Speed Scale=20
   ) {enabled=false, module=AddVelocity002}
   &AddVelocity(Velocity=&RandomRangeVector(Minimum=<0 -60 -85>, Maximum=<0 60 85>), Velocity Speed Scale=2) {module=AddVelocity001}
  ]
  particleUpdateScript [
   &ParticleState()
   &DynamicMaterialParameters(
    Index 0 Param 1=0.2
    Index 0 Param 2=50
    Module.Index0Param0Write=true
    Module.Index0Param1Write=true
    Module.Index0Param2Write=true
    Module.Index0Param3Write=true
   )
   &ScaleVelocity(Velocity Scale=&RandomRangeVector(
     Minimum=&VectorFromCurve(VectorCurve=DataInterface() {propertyValues {XCurve {keys [<at 0 1>,<at 1 0>]}, YCurve {keys [<at 0 1>,<at 1 0>]}, ZCurve {keys [<at 0 1>,<at 1 0>]}}})
     Maximum=&VectorFromCurve(VectorCurve=DataInterface() {propertyValues {XCurve {keys [<at 0 1>,<at 1 0>]}, YCurve {keys [<at 0 1>,<at 1 0>]}, ZCurve {keys [<at 0 1>,<at 1 0>]}}})
    )) {enabled=false}
   &SolveForcesAndVelocity(Clamp Velocity=false, Limit Acceleration=true)
   &ScaleColor(
    Scale Alpha=&FloatFromCurve(FloatCurve=DataInterface() {propertyValues {Curve {keys [<at 0 1.5>,<at 0.5 1>,<at 1 0>]}}})
    Scale RGB=&VectorFromCurve(
     VectorCurve=DataInterface() {propertyValues {XCurve {keys [<at 0 1>,<at 0.5 1>]}, YCurve {keys [<at 0 1>,<at 0.5 0>]}, ZCurve {keys [<at 0 1>,<at 0.5 -1>]}}}
     Scale Curve=<0 0 0>
    )
    ScaleA=true
    ScaleRGB=true
   )
   &ScaleRibbonWidth(Ribbon Width Scale=&FloatFromCurve(FloatCurve=DataInterface() {propertyValues {Curve {keys [<at 0 1>,<at 1 0>]}}}))
   &CameraOffset(Camera Offset Amount=&RandomRangeFloat(Minimum=-0.5, Maximum=0.5))
  ]
  rendererClasses [&NiagaraRibbonRendererProperties]
  renderers [{rendererClass=&NiagaraRibbonRendererProperties}]
 }
 CoreLightningTrail {
  emitterUpdateScript [
   &EmitterState(Life Cycle Mode=::NewEnumerator1, Loop Duration=4, UseLoopDelay=false)
   &SpawnRate(SpawnRate=60, Use Spawn Probability=false)
  ]
  particleSpawnScript [
   &InitializeParticle(
    Color=<rgba 2 0.1 0.05 5>
    Lifetime=0.5
    Position Offset=<-186.9681 0 0>
    Ribbon Width=250
    Ribbon Width Mode=::NewEnumerator1
    UsePositionOffset=true
    Write Material Random=true
   )
   &AddVelocity(Velocity=<800 50 0>, Velocity Speed Scale=4)
   &AddVelocity(Velocity=&RandomRangeVector(Minimum=<0 -50 -50>, Maximum=<0 50 50>)) {enabled=false, module=AddVelocity001}
  ]
  particleUpdateScript [
   &DynamicMaterialParameters(
    Index 0 Param 1=&RandomRangeFloat(Minimum=2, Maximum=5)
    Index 0 Param 2=&RandomRangeFloat(Minimum=0, Maximum=1)
    Index 0 Param 3=0.1
    Index 0 Param 4=0.3
    Module.Index0Param0Write=true
    Module.Index0Param1Write=true
    Module.Index0Param2Write=true
    Module.Index0Param3Write=true
   )
   &ParticleState()
   &ScaleVelocity(Velocity Scale=&RandomRangeVector(
     Minimum=&VectorFromCurve(VectorCurve=DataInterface() {propertyValues {XCurve {keys [<at 0 1>,<at 1 0>]}, YCurve {keys [<at 0 1>,<at 1 0>]}, ZCurve {keys [<at 0 1>,<at 1 0>]}}})
     Maximum=&VectorFromCurve(VectorCurve=DataInterface() {propertyValues {XCurve {keys [<at 0 1>,<at 1 0>]}, YCurve {keys [<at 0 1>,<at 1 0>]}, ZCurve {keys [<at 0 1>,<at 1 0>]}}})
    )) {enabled=false}
   &SolveForcesAndVelocity(Clamp Velocity=false, Limit Acceleration=true)
   &ScaleColor(
    Scale Alpha=&FloatFromCurve(FloatCurve=DataInterface() {propertyValues {Curve {keys [
         {interpMode=RCIM_Cubic, tangentMode=RCTM_Break, time=0, value=5}
         {interpMode=RCIM_Cubic, tangentMode=RCTM_Break, time=0.25, value=1}
         {interpMode=RCIM_Cubic, tangentMode=RCTM_Break, time=1, value=0}
        ]}}})
    Scale RGB=&VectorFromCurve(VectorCurve=DataInterface() {propertyValues {
       XCurve {keys [{interpMode=RCIM_Cubic, tangentMode=RCTM_Break, time=0, value=1}, <at 1 0>]}
       YCurve {keys [<at 0 0>,<at 0 0>]}
       ZCurve {keys [<at 0 0>,<at 1 0>]}
      }})
    ScaleA=true
    ScaleRGB=true
   )
   &ScaleRibbonWidth(Ribbon Width Scale=&FloatFromCurve(FloatCurve=DataInterface() {propertyValues {Curve {keys [<at 0 1>,<at 1 0>]}}}))
  ]
  rendererClasses [&NiagaraRibbonRendererProperties]
  renderers [{rendererClass=&NiagaraRibbonRendererProperties}]
 }
 CoreTrail {
  emitterUpdateScript [
   &EmitterState(Life Cycle Mode=::NewEnumerator1, Loop Duration=4, UseLoopDelay=false)
   &SpawnRate(SpawnRate=60, Use Spawn Probability=false)
  ]
  particleSpawnScript [
   &InitializeParticle(
    Color=<rgba 1 0.1 1.8 5>
    Position Offset=<-210.4285 0 0>
    Ribbon Width=400
    Ribbon Width Mode=::NewEnumerator1
    UsePositionOffset=true
    Write Material Random=true
   )
   &AddVelocity(Velocity=<800 50 0>, Velocity Speed Scale=3)
   &AddVelocity(Velocity=&RandomRangeVector(Minimum=<0 -50 -50>, Maximum=<0 50 50>)) {enabled=false, module=AddVelocity001}
  ]
  particleUpdateScript [
   &DynamicMaterialParameters(
    Index 0 Param 1=3
    Index 0 Param 2=&RandomRangeFloat(Minimum=0, Maximum=1)
    Index 0 Param 4=0.5
    Module.Index0Param0Write=true
    Module.Index0Param1Write=true
    Module.Index0Param2Write=true
    Module.Index0Param3Write=true
   )
   &ParticleState()
   &ScaleVelocity(Velocity Scale=&RandomRangeVector(
     Minimum=&VectorFromCurve(VectorCurve=DataInterface() {propertyValues {XCurve {keys [<at 0 1>,<at 1 0>]}, YCurve {keys [<at 0 1>,<at 1 0>]}, ZCurve {keys [<at 0 1>,<at 1 0>]}}})
     Maximum=&VectorFromCurve(VectorCurve=DataInterface() {propertyValues {XCurve {keys [<at 0 1>,<at 1 0>]}, YCurve {keys [<at 0 1>,<at 1 0>]}, ZCurve {keys [<at 0 1>,<at 1 0>]}}})
    )) {enabled=false}
   &SolveForcesAndVelocity(Clamp Velocity=false, Limit Acceleration=true)
   &ScaleColor(
    Scale Alpha=&FloatFromCurve(FloatCurve=DataInterface() {propertyValues {Curve {keys [
         {interpMode=RCIM_Cubic, tangentMode=RCTM_Break, time=0, value=5}
         {interpMode=RCIM_Cubic, tangentMode=RCTM_Break, time=0.5528, value=1}
         {interpMode=RCIM_Cubic, tangentMode=RCTM_Break, time=0.9909, value=0}
        ]}}})
    Scale RGB=&VectorFromCurve(VectorCurve=DataInterface() {propertyValues {
       XCurve {keys [{interpMode=RCIM_Cubic, tangentMode=RCTM_Break, time=0, value=0.05}, <at 0.25 0.01>]}
       YCurve {keys [{interpMode=RCIM_Cubic, tangentMode=RCTM_Break, time=0, value=0.01}, <at 0.25 0.01>]}
       ZCurve {keys [{interpMode=RCIM_Cubic, tangentMode=RCTM_Break, time=0, value=0.1}, <at 0.25 0.05>]}
      }})
    ScaleA=true
    ScaleRGB=true
   )
   &ScaleRibbonWidth(Ribbon Width Scale=&FloatFromCurve(FloatCurve=DataInterface() {propertyValues {Curve {keys [<at 0 1>,<at 1 0>]}}}))
  ]
  rendererClasses [&NiagaraRibbonRendererProperties]
  renderers [{rendererClass=&NiagaraRibbonRendererProperties}]
 }
 Lightning {
  emitterUpdateScript [
   &EmitterState(Life Cycle Mode=::NewEnumerator1, Loop Duration=4, UseLoopDelay=false)
   &SpawnRate(SpawnRate=20, Use Spawn Probability=false)
  ]
  particleSpawnScript [
   &InitializeParticle(
    Color=&RandomRangeLinearColor(Minimum=<rgba -4 -2 -4 50>, Maximum=<rgba 4 2 4 50>)
    Lifetime=0.65
    Position Offset=<-250 0 0>
    Ribbon Width=&RandomRangeFloat(Minimum=150, Maximum=170)
    Ribbon Width Mode=::NewEnumerator1
    UsePositionOffset=true
    Write Material Random=true
   )
   &AddVelocity(Velocity=<800 0 0>, Velocity Speed Scale=3.5)
   &AddVelocity(
    Velocity=&VectorFromCurve(VectorCurve=DataInterface() {propertyValues {XCurve {keys [<at 0 0>,<at 1 20>]}, YCurve {keys [<at 0 0>,<at 1 20>]}, ZCurve {keys [<at 0 0>,<at 1 20>]}}})
    Velocity Speed Scale=20
   ) {enabled=false, module=AddVelocity002}
   &AddVelocity(Velocity=&RandomRangeVector(Minimum=<0 -60 -85>, Maximum=<0 60 85>), Velocity Speed Scale=2) {enabled=false, module=AddVelocity001}
  ]
  particleUpdateScript [
   &ParticleState()
   &DynamicMaterialParameters(
    Index 0 Param 1=0.5
    Index 0 Param 2=50
    Module.Index0Param0Write=true
    Module.Index0Param1Write=true
    Module.Index0Param2Write=true
    Module.Index0Param3Write=true
   )
   &ScaleVelocity(Velocity Scale=&RandomRangeVector(
     Minimum=&VectorFromCurve(VectorCurve=DataInterface() {propertyValues {XCurve {keys [<at 0 1>,<at 1 0>]}, YCurve {keys [<at 0 1>,<at 1 0>]}, ZCurve {keys [<at 0 1>,<at 1 0>]}}})
     Maximum=&VectorFromCurve(VectorCurve=DataInterface() {propertyValues {XCurve {keys [<at 0 1>,<at 1 0>]}, YCurve {keys [<at 0 1>,<at 1 0>]}, ZCurve {keys [<at 0 1>,<at 1 0>]}}})
    )) {enabled=false}
   &SolveForcesAndVelocity(Clamp Velocity=false, Limit Acceleration=true)
   &ScaleColor(
    Scale Alpha=&FloatFromCurve(FloatCurve=DataInterface() {propertyValues {Curve {keys [<at 0 1.5>,<at 0.5 1>,<at 1 0>]}}})
    Scale RGB=&VectorFromCurve(VectorCurve=DataInterface() {propertyValues {XCurve {keys [<at 0 1>,<at 0.5 1>]}, YCurve {keys [<at 0 1>,<at 0.5 0>]}, ZCurve {keys [<at 0 1>,<at 0.5 -1>]}}})
    ScaleA=true
    ScaleRGB=true
   )
   &ScaleRibbonWidth(Ribbon Width Scale=&FloatFromCurve(FloatCurve=DataInterface() {propertyValues {Curve {keys [<at 0 1>,<at 1 0>]}}}))
   &CameraOffset(Camera Offset Amount=&RandomRangeFloat(Minimum=-0.5, Maximum=0.5))
  ]
  rendererClasses [&NiagaraRibbonRendererProperties]
  renderers [{rendererClass=&NiagaraRibbonRendererProperties}]
 }
 Menu_Embers {
  emitterUpdateScript [
   &EmitterState()
   &SpawnRate(SpawnRate=20, Use Spawn Probability=true)
   &SpawnBurst_Instantaneous(Spawn Count=15, Use Spawn Probability=true, UseLoopCountLimit=false) {enabled=false}
  ]
  particleSpawnScript [
   &InitializeParticle(
    Color=<rgba 2 0.1 0.05 5>
    Lifetime Max=0.8
    Lifetime Min=0.5
    Lifetime Mode=::NewEnumerator1
    Position Offset=<-50 30 -50>
    Sprite Rotation Mode=::NewEnumerator2
    Sprite Size Mode=::NewEnumerator1
    Uniform Sprite Size Max=7.5
    UsePositionOffset=true
    Write Material Random=true
   )
   &AddVelocity(Velocity=<1000 0 0>, Velocity Speed Scale=7) {module=AddVelocity001}
   &AddVelocity(
    Rotation Coordinate Space=::World
    Velocity=&RandomRangeVector(Minimum=<0 0 -100>, Maximum=<0 0 100>)
    Velocity Speed Scale=3
   )
   &ShapeLocation(MaskSpawnGroup=false, Override Seed=false, Shape Primitive=::NewEnumerator5, UseOwnerScale=true)
  ]
  particleUpdateScript [
   &ParticleState()
   &DynamicMaterialParameters(
    Index 0 Param 1=&RandomRangeFloat(Minimum=0, Maximum=1)
    Index 0 Param 2=0.1
    Index 0 Param 4=0.5
    Module.Index0Param0Write=true
    Module.Index0Param1Write=true
    Module.Index0Param2Write=true
    Module.Index0Param3Write=true
   ) {enabled=false}
   &ScaleVelocity(Velocity Scale=&VectorFromFloat(Value=&FloatFromCurve(FloatCurve=DataInterface() {propertyValues {Curve {keys [<at 0 0.5>,<at 0.2 0.2>]}}})))
   &SpriteRotationRate(Rotation Rate=&RandomRangeFloat(Minimum=-10, Maximum=10))
   &SolveForcesAndVelocity(Clamp Velocity=false, Limit Acceleration=false)
   &ScaleColor(Scale RGB=<15 8 8>, ScaleA=true, ScaleRGB=true)
   &ScaleSpriteSize(Uniform Curve Sprite Scale=DataInterface() {propertyValues {Curve {keys [{arriveTangent=0.3218, interpMode=RCIM_Cubic, leaveTangent=0.3218, tangentMode=RCTM_User, time=0, value=1}, <at 1 3 RCIM_Cubic>]}}})
  ]
  rendererClasses [&NiagaraSpriteRendererProperties]
  renderers [{rendererClass=&NiagaraSpriteRendererProperties}]
 }
}
summary {
 systemName=PowerShot_PS
 emitters [
  {emitterName=CoreTrail, rendererClasses [&NiagaraRibbonRendererProperties]}
  {emitterName=CoreLightningTrail, rendererClasses [&NiagaraRibbonRendererProperties]}
  {emitterName=CoreGlow001, rendererClasses [&NiagaraRibbonRendererProperties]}
  {emitterName=Lightning, rendererClasses [&NiagaraRibbonRendererProperties]}
  {emitterName=Menu_Embers, rendererClasses [&NiagaraSpriteRendererProperties]}
  {emitterName=BaseSparks01, rendererClasses [&NiagaraSpriteRendererProperties]}
  {emitterName=BaseSparks01001, rendererClasses [&NiagaraSpriteRendererProperties]}
  {emitterName=BaseCoreTrail, rendererClasses [&NiagaraRibbonRendererProperties]}
 ]
 userVariables []
}
