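# Translucent unlit Niagara sprite material for lightning in the Powershot boost vehicle cosmetic.
# Samples a trail texture as a time-driven flipbook, with static switches for rotation, channel selection or blending, and an optional panning mask overlay.
# A U-coordinate gradient fades both ends, with emissive weighted toward the start.
# Particle color and Dynamic Parameters drive tint, brightness and blend, feeding an emissive adjustment function and opacity with optional depth fade.
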Material /VehicleCosmetics/Boosts/Boost_Powershot/Material/MM_Lightning
graph {
 properties {bUsedWithNiagaraSprites=true, blendMode=BLEND_Translucent, pixelDepthOffsetMode=PDOM_Legacy, shadingModel=MSM_Unlit, usedShadingModels=Unlit}
 nodes [
  n1 = ParticleColor()
  n2 = TextureCoordinate() {uTiling=0.05, vTiling=0.25}
  n3 = Multiply(A=Time(), B=ScalarParameter() {defaultValue=1, parameterName=PanSpeed})
  n4 = StaticSwitchParameter(
   True=Rotator(Coordinate=Panner(Coordinate=@n2, Time=@n3) {speedY=-0.4}, Time=Constant() {r=0.5}) {speed=3.14}
   False=Panner(Coordinate=TextureCoordinate() {uTiling=0.25, vTiling=0.05}, Time=@n3) {speedX=-0.4}
  ) {group=FlipBook, parameterName=RotateFlipbook}
  n5 = Panner(Coordinate=TextureCoordinate() {desc=1x1, vTiling=0.5}, Time=Constant() {r=1}) {speedY=0.25}
  n6 = Constant() {r=0.5}
  n7 = ScalarParameter() {defaultValue=4, group=FlipBookss, parameterName=Columns}
  n8 = ScalarParameter() {defaultValue=4, group=FlipBook, parameterName=Rows}
  n9 = AppendVector(A=@n7, B=@n8)
  n10 = Divide(
   A=StaticSwitchParameter(True=Rotator(Coordinate=@n5, Time=@n6) {speed=3.14}, False=@n5) {group=FlipBook, parameterName=RotateFlipbook}
   B=@n9
  )
  n11 = AppendVector(A=Multiply(A=@n7, B=@n8), B=@n8)
  n12 = Frac(None=Multiply(A=Time(), B=ScalarParameter() {defaultValue=1, group=FlipBook, parameterName=Speed}))
  n13 = AppendVector(A=@n12, B=@n12)
  n14 = Divide(A=Floor(None=Multiply(A=@n11, B=@n13)), B=@n9)
  n15 = TextureCoordinate() {coordinateIndex=1, desc=1x1}
  n16 = Rotator(Coordinate=@n15, Time=Constant() {r=0.5}) {speed=3.14}
  n17 = TextureSampleParameter2D(UVs=StaticSwitchParameter(True=Add(A=@n10, B=@n14), False=StaticSwitchParameter(True=@n16, False=@n15) {group=FlipBook, parameterName=RotateFlipbook}) {defaultValue=true, group=FlipBook, parameterName=FlipBook}) {group=FlipBook, parameterName=BaseTexture, texture=/VehicleCosmetics/Boosts/Boost_Powershot/Textures/ZTI_Trail_RGB_FB_T}
  n18 = StaticComponentMaskParameter(None=@n17.RGB) {defaultG=true, group=Blend, parameterName=Blend_01}
  n19 = StaticComponentMaskParameter(None=@n17.RGB) {defaultB=true, group=Blend, parameterName=Blend_02}
  n20 = DynamicParameter() {paramNames [Brightness,LerpBlend,Param3,Param4]}
  n21 = Multiply(A=@n18, B=@n19)
  n22 = Add(A=@n18, B=@n19)
  n23 = StaticSwitchParameter(
   True=StaticSwitchParameter(
    True=LinearInterpolate(A=@n18, B=@n19, Alpha=@n20.LerpBlend)
    False=StaticSwitchParameter(True=@n21, False=@n22) {group=Blend, parameterName="BlendType(T=M)(F=A)"}
   ) {group=Blend, parameterName=LerpBlend}
   False=StaticComponentMaskParameter(None=@n17.RGB) {defaultG=true, group=Blend, parameterName=RGB_Mask}
  ) {group=Blend, parameterName=Blend}
  n24 = Multiply(
   A=Multiply(
    A=ScalarParameter() {defaultValue=3, group=Pan, parameterName=PanBrightness}
    B=TextureSample(UVs=@n4) {texture=/VehicleCosmetics/Boosts/Boost_Powershot/Textures/Time_Mask_T}.RGB
   )
   B=@n23
  )
  n25 = StaticSwitchParameter(True=StaticSwitchParameter(True=Add(A=@n24, B=@n23), False=@n24) {group=Pan, parameterName=PanAdd}, False=@n23) {group=Pan, parameterName=PanLightning}
  n26 = TextureCoordinate() {uTiling=4, vTiling=4}
  n27 = ComponentMask(None=StaticSwitchParameter(True=TextureCoordinate(), False=Frac(None=@n26)) {defaultValue=true, group=FlipBook, parameterName=FlipBook}) {r=true}
  n28 = OneMinus(None=@n27)
  n29 = Multiply(A=Multiply(A=Multiply(A=@n25, B=@n27), B=@n20.Brightness), B=@n28)
  n30 = Multiply(A=@n1.A, B=@n29)
  n31 = Constant() {r=100}
  Clamp(None=StaticSwitchParameter(
    True=MaterialFunctionCall(Opacity In=@n30, DepthFade Distance=@n31) {materialFunction=&MF_VCVFX_DepthFade}.Result
    False=@n30
   ) {group=Opacity, parameterName=DBA?})
  n32 = Multiply(A=@n28) {constB=8}
  MaterialFunctionCall(Emissive=Multiply(A=@n1.RGB, B=Multiply(A=@n32, B=@n29))) {materialFunction=&MF_FX_EmissiveAdjustments}
 ]
}
