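# Translucent unlit lightning arc material for the Powershot boost's Niagara ribbons and sprites.
# Zero-tiled panned cloud samples act as time-varying scalars, blending lightning texture channels and flickering their intensity.
# That mask drives squared-color emissive scaled by dynamic Brightness, and an alpha term with particle alpha and U-axis fade.
# A switchable TrailBend builds a sine/cosine offset, using particle RGB as frequency, speed and amplitude.
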Material /VehicleCosmetics/Boosts/Boost_Powershot/Material/MM_LightningArc
graph {
 properties {
  bUsedWithNiagaraRibbons=true
  bUsedWithNiagaraSprites=true
  blendMode=BLEND_Translucent
  pixelDepthOffsetMode=PDOM_Legacy
  shadingModel=MSM_Unlit
  usedShadingModels=Unlit
 }
 nodes [
  n1 = VectorParameter() {parameterName=Color, defaultValue=<rgba 0.1 0.25 2 1>}
  n2 = DynamicParameter() {paramNames [Parametro1,Brightness,Param3,Param4]}
  n3 = TextureSampleParameter2D(UVs=TextureCoordinate() {coordinateIndex=1}) {parameterName=LightningTexture, texture=/VehicleCosmetics/Boosts/Boost_Powershot/Textures/LightningTrail_Mix_T}
  n4 = TextureCoordinate() {coordinateIndex=1, uTiling=0, vTiling=0}
  n5 = AppendVector(
   A=TextureSample(UVs=Panner(Coordinate=@n4) {speedX=0.5, speedY=0.325}) {texture=/VehicleCosmetics/Boosts/Boost_Powershot/Textures/Clouds_RGB_Soft_03_T}.G
   B=TextureSample(UVs=Panner(Coordinate=@n4) {speedX=-0.5, speedY=-0.15}) {texture=&Clouds_RGB_Soft_03_T}.G
  )
  n6 = Clamp(None=Multiply(A=Subtract(A=@n5) {constB=0.2}) {constB=5})
  n7 = LinearInterpolate(A=@n3.R, B=@n3.R, Alpha=ComponentMask(None=@n6) {r=true})
  n8 = ComponentMask(None=@n6) {g=true}
  n9 = Panner(Coordinate=TextureCoordinate() {coordinateIndex=1, uTiling=0, vTiling=0}, Time=Multiply(A=@n2.Parametro1, B=Time())) {speedX=1, speedY=1.5}
  n10 = Multiply(A=Subtract(A=TextureSample(UVs=@n9) {texture=&Clouds_RGB_Soft_03_T}.R, B=Constant() {r=0.4}) {constB=0.4}) {constB=5}
  n11 = Clamp(None=Multiply(A=LinearInterpolate(A=@n7, B=@n3.B, Alpha=@n8), B=Clamp(None=@n10)))
  MaterialFunctionCall(Emissive=Multiply(A=Multiply(A=@n1.RGB, B=@n1.RGB), B=Multiply(A=@n2.Brightness, B=@n11))) {materialFunction=&MF_FX_EmissiveAdjustments}
  n12 = TextureCoordinate() {uTiling=1.01}
  n13 = Multiply(A=OneMinus(None=ComponentMask(None=@n12) {r=true}))
  Multiply(A=ParticleColor().A, B=Multiply(A=@n11, B=Clamp(None=@n13)))
  n14 = ParticleColor()
  n15 = Multiply(A=@n14.G, B=OneMinus(None=Time()))
  n16 = ComponentMask(None=Multiply(A=TextureCoordinate() {coordinateIndex=1, uTiling=2}, B=@n14.R)) {r=true}
  n17 = Add(A=@n15, B=@n16)
  n18 = Add(A=Cosine(None=@n17), B=Cosine(None=@n17) {period=2})
  n19 = Add(A=@n16, B=@n15)
  n20 = Add(A=Sine(None=@n19), B=Sine(None=@n19) {period=2})
  n21 = Transform(None=AppendVector(A=AppendVector(A=@n18, B=@n20), B=@n18)) {transformType=TRANSFORM_Local}
  n22 = Add(A=ComponentMask(None=TextureCoordinate() {uTiling=6}) {r=true}) {constB=0.35}
  StaticSwitchParameter(True=Multiply(A=Multiply(A=@n21, B=@n14.B), B=Clamp(None=@n22)), False=Constant()) {defaultValue=true, group=TrailBend, parameterName=TrailBend}
 ]
}
